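# Conversation database for the Bounty Board station NPC, entered through the BountyBoardStation entry tag.
# A greeting speech line offers three choices: go back, take a player bounty, or buy chest intel.
# The intel choice uses a treasure-chest tag filter to find a limited number of nearby containers and mark them with details for the squad at a price.
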ConversationDatabase /NPCLibrary/NPCs/NPCLibrary_Conversation_BountyBoard
entry_points [{tag=Conversation.Entry.BountyBoardStation, destinations [#1]}]
nodes [
 ConversationEntryPointNode_7 {
  id=#1
  class=ConversationEntryPointNode
  outputs [#2]
  properties {entryTag {tagName=Conversation.Entry.BountyBoardStation}}
 }
 Speech_C_4 {
  id=#2
  class=Speech_C
  outputs [#3, #4, #5]
  properties {nodeName="", generalConfig {defaultMessage="Spend all those bars in one place!"}}
 }
 GrantIntel_C_0 {
  id=#5
  class=GrantIntel_C
  outputs []
  properties {filterDatas [{
     classFilter=/Script/FortniteGame.BuildingContainer
     maxRadius {registryType=None, value=10000, curve {curveTable=null, rowName=None}}
     maxToFind {registryType=None, value=5, curve {curveTable=null, rowName=None}}
     objectTypes []
     price {registryType=None, value=50, curve {curveTable=null, rowName=None}}
     rarity {registryType=None, value=6, curve {curveTable=null, rowName=None}}
     showMarkerDetails {registryType=None, value=1, curve {curveTable=null, rowName=None}}
     showToSquad {registryType=None, value=1, curve {curveTable=null, rowName=None}}
     targetFilter {
      autoDescription=" ALL( Building.Type.Container.TreasureChest )"
      tokenStreamVersion=0
      userDescription=""
      queryTokenStream [0,1,2,1,0]
      tagDictionary [{tagName=Building.Type.Container.TreasureChest}]
     }
    }]}
  sub_nodes [FortConversationChoiceWithPreviewNode_0 {class=FortConversationChoiceWithPreviewNode, properties {defaultChoiceDisplayText=Chests}}]
 }
 GrantPlayerBounty_C_0 {
  id=#4
  class=GrantPlayerBounty_C
  outputs []
  sub_nodes [ConversationChoiceNode_0 {class=ConversationChoiceNode, properties=~}]
 }
 Back_C_5 {
  id=#3
  class=Back_C
  sub_nodes [ConversationChoiceNode_4 {class=ConversationChoiceNode, properties {defaultChoiceDisplayText=BACK}}]
 }
]
