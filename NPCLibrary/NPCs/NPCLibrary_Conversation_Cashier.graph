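# Conversation database for the cashier NPC in NPCLibrary, defining the dialogue that starts at the Conversation.Entry.Cashier entry tag.
# The entry point node leads to a single Speech node, which has no outputs, so the conversation is one step with no choices or branching.
# The Speech node carries an empty name and no icon, and no dialogue text is set in the database.
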ConversationDatabase /NPCLibrary/NPCs/NPCLibrary_Conversation_Cashier
entry_points [{tag=Conversation.Entry.Cashier, destinations [#1]}]
nodes [
 ConversationEntryPointNode_7 {
  id=#1
  class=ConversationEntryPointNode
  outputs [#2]
  properties {entryTag {tagName=Conversation.Entry.Cashier}}
 }
 Speech_C_4 {id=#2, class=Speech_C, outputs [], properties {nodeName="", softTaskIcon=null}}
]
