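# Conversation database for the Cashier NPC's quest dialogue, offering slotted quests on behalf of the cashier's client.
# The entry point fans out to five near-identical Speech branches, one per Tandem Quest service.
# Each is gated on that service, prerequisites, no completion this match, no active provider quest, and no recorded GrantedQuest event with this NPC.
# Each leads to Accept, which grants the slotted quest with service-used effects, or Back.
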ConversationDatabase /NPCLibrary/NPCs/NPCLibrary_Conversation_Cashier_Quests
entry_points [{tag=Conversation.Entry.Cashier.Quest, destinations [#1]}]
nodes [
 ConversationEntryPointNode_0 {
  id=#1
  class=ConversationEntryPointNode
  outputs [
   #2
   #3
   #4
   #5
   #6
  ]
  properties {entryTag {tagName=Conversation.Entry.Cashier.Quest}}
 }
 Speech_C_5 {
  id=#6
  class=Speech_C
  outputs [#7, #8]
  properties {
   softTaskIcon=/Game/Athena/HUD/Interaction/T_UI_NPC_quest.T_UI_NPC_Quest
   generalConfig {contextualMessages [
     {message=Can you help my client with this?, contextRequirements [], priority {value=0, curve {curveTable=null, rowName=None}}}
     {message=My client needs a hand. Or two., contextRequirements [], priority {value=0, curve {curveTable=null, rowName=None}}}
     {message=Help my client. Help me. Get Paid., contextRequirements [], priority {value=0, curve {curveTable=null, rowName=None}}}
     {
      message=At least my client's basement isn't full of Llamas again...
      contextRequirements []
      priority {value=0, curve {curveTable=null, rowName=None}}
     }
     {message=My client requested your services...., contextRequirements [], priority {value=0, curve {curveTable=null, rowName=None}}}
    ]}
  }
  sub_nodes [
   FortConversationRequirement_HasService_0 {class=FortConversationRequirement_HasService, properties {serviceTag {tagName=Tandem.Service.Quest5}}}
   FortConversationChoiceWithPreviewNode_0 {class=FortConversationChoiceWithPreviewNode, properties {defaultChoiceDisplayText=Quest}}
   FortConversationRequirement_SlottedQuestNotCompletedThisMatch_0 {class=FortConversationRequirement_SlottedQuestNotCompletedThisMatch, properties {questSlot=5}}
   FortConversationRequirement_AllSlottedQuestPrerequisitesCompleted_0 {class=FortConversationRequirement_AllSlottedQuestPrerequisitesCompleted, properties {questSlot=5}}
   FortConversationRequirement_HasNoActiveQuestsFromSlottedQuestProvider_0 {class=FortConversationRequirement_HasNoActiveQuestsFromSlottedQuestProvider, properties {questSlot=5}}
   FortConversationRequirement_ControllerRequirement_4 {
    class=FortConversationRequirement_ControllerRequirement
    properties {
     nodeName=Require: Has Recorded Conversation Event With NPC
     requirement {
      class=FortControllerRequirement_HasRecordedConversationEventWithNPC
      properties {bInverseCondition=true, queryMethod=CheckAgainstCurrentConversationParticipant, queryEventTag {tagName=Conversation.Event.GrantedQuest}}
     }
    }
   }
  ]
 }
 Back_C_3 {
  id=#8
  class=Back_C
  sub_nodes [ConversationChoiceNode_0 {class=ConversationChoiceNode, properties {defaultChoiceDisplayText="BACK!"}}]
 }
 GrantSlottedQuest_C_5 {
  id=#7
  class=GrantSlottedQuest_C
  outputs []
  properties {bNotifyQuestProvider=true, bShouldDisplayQuestProviderImage=true, bShouldOverrideSpeakerText=false, questSlot=5}
  sub_nodes [
   ConversationChoiceNode_1 {class=ConversationChoiceNode, properties {defaultChoiceDisplayText="ACCEPT!"}}
   SideEffect_PlayServiceUsedEffects_C_3 {class=SideEffect_PlayServiceUsedEffects_C}
  ]
 }
 Speech_C_6 {
  id=#5
  class=Speech_C
  outputs [#9, #10]
  properties {
   softTaskIcon=&T_UI_NPC_Quest
   generalConfig {contextualMessages [
     {message=Can you help my client with this?, contextRequirements [], priority {value=0, curve {curveTable=null, rowName=None}}}
     {message=My client needs a hand. Or two., contextRequirements [], priority {value=0, curve {curveTable=null, rowName=None}}}
     {message=Help my client. Help me. Get Paid., contextRequirements [], priority {value=0, curve {curveTable=null, rowName=None}}}
     {
      message=At least my client's basement isn't full of Llamas again...
      contextRequirements []
      priority {value=0, curve {curveTable=null, rowName=None}}
     }
     {message=My client requested your services...., contextRequirements [], priority {value=0, curve {curveTable=null, rowName=None}}}
    ]}
  }
  sub_nodes [
   FortConversationRequirement_HasService_1 {class=FortConversationRequirement_HasService, properties {serviceTag {tagName=Tandem.Service.Quest4}}}
   FortConversationChoiceWithPreviewNode_1 {class=FortConversationChoiceWithPreviewNode, properties {defaultChoiceDisplayText=Quest}}
   FortConversationRequirement_SlottedQuestNotCompletedThisMatch_1 {class=FortConversationRequirement_SlottedQuestNotCompletedThisMatch, properties {questSlot=4}}
   FortConversationRequirement_AllSlottedQuestPrerequisitesCompleted_1 {class=FortConversationRequirement_AllSlottedQuestPrerequisitesCompleted, properties {questSlot=4}}
   FortConversationRequirement_HasNoActiveQuestsFromSlottedQuestProvider_1 {class=FortConversationRequirement_HasNoActiveQuestsFromSlottedQuestProvider, properties {questSlot=4}}
   FortConversationRequirement_ControllerRequirement_3 {
    class=FortConversationRequirement_ControllerRequirement
    properties {
     nodeName=Require: Has Recorded Conversation Event With NPC
     requirement {
      class=FortControllerRequirement_HasRecordedConversationEventWithNPC
      properties {bInverseCondition=true, queryMethod=CheckAgainstCurrentConversationParticipant, queryEventTag {tagName=Conversation.Event.GrantedQuest}}
     }
    }
   }
  ]
 }
 Back_C_6 {
  id=#10
  class=Back_C
  sub_nodes [ConversationChoiceNode_2 {class=ConversationChoiceNode, properties {defaultChoiceDisplayText="BACK!"}}]
 }
 GrantSlottedQuest_C_6 {
  id=#9
  class=GrantSlottedQuest_C
  outputs []
  properties {bNotifyQuestProvider=true, bShouldDisplayQuestProviderImage=true, bShouldOverrideSpeakerText=false, questSlot=4}
  sub_nodes [
   ConversationChoiceNode_3 {class=ConversationChoiceNode, properties {defaultChoiceDisplayText="ACCEPT!"}}
   SideEffect_PlayServiceUsedEffects_C_4 {class=SideEffect_PlayServiceUsedEffects_C}
  ]
 }
 Speech_C_7 {
  id=#4
  class=Speech_C
  outputs [#11, #12]
  properties {
   softTaskIcon=&T_UI_NPC_Quest
   generalConfig {contextualMessages [
     {message=Can you help my client with this?, contextRequirements [], priority {value=0, curve {curveTable=null, rowName=None}}}
     {message=My client needs a hand. Or two., contextRequirements [], priority {value=0, curve {curveTable=null, rowName=None}}}
     {message=Help my client. Help me. Get Paid., contextRequirements [], priority {value=0, curve {curveTable=null, rowName=None}}}
     {
      message=At least my client's basement isn't full of Llamas again...
      contextRequirements []
      priority {value=0, curve {curveTable=null, rowName=None}}
     }
     {message=My client requested your services...., contextRequirements [], priority {value=0, curve {curveTable=null, rowName=None}}}
    ]}
  }
  sub_nodes [
   FortConversationRequirement_HasService_2 {class=FortConversationRequirement_HasService, properties {serviceTag {tagName=Tandem.Service.Quest3}}}
   FortConversationChoiceWithPreviewNode_2 {class=FortConversationChoiceWithPreviewNode, properties {defaultChoiceDisplayText=Quest}}
   FortConversationRequirement_SlottedQuestNotCompletedThisMatch_2 {class=FortConversationRequirement_SlottedQuestNotCompletedThisMatch, properties {questSlot=3}}
   FortConversationRequirement_AllSlottedQuestPrerequisitesCompleted_2 {class=FortConversationRequirement_AllSlottedQuestPrerequisitesCompleted, properties {questSlot=3}}
   FortConversationRequirement_HasNoActiveQuestsFromSlottedQuestProvider_2 {class=FortConversationRequirement_HasNoActiveQuestsFromSlottedQuestProvider, properties {questSlot=3}}
   FortConversationRequirement_ControllerRequirement_2 {
    class=FortConversationRequirement_ControllerRequirement
    properties {
     nodeName=Require: Has Recorded Conversation Event With NPC
     requirement {
      class=FortControllerRequirement_HasRecordedConversationEventWithNPC
      properties {bInverseCondition=true, queryMethod=CheckAgainstCurrentConversationParticipant, queryEventTag {tagName=Conversation.Event.GrantedQuest}}
     }
    }
   }
  ]
 }
 Back_C_7 {
  id=#12
  class=Back_C
  sub_nodes [ConversationChoiceNode_4 {class=ConversationChoiceNode, properties {defaultChoiceDisplayText="BACK!"}}]
 }
 GrantSlottedQuest_C_7 {
  id=#11
  class=GrantSlottedQuest_C
  outputs []
  properties {bNotifyQuestProvider=true, bShouldDisplayQuestProviderImage=true, bShouldOverrideSpeakerText=false, questSlot=3}
  sub_nodes [
   ConversationChoiceNode_5 {class=ConversationChoiceNode, properties {defaultChoiceDisplayText="ACCEPT!"}}
   SideEffect_PlayServiceUsedEffects_C_2 {class=SideEffect_PlayServiceUsedEffects_C}
  ]
 }
 Speech_C_9 {
  id=#3
  class=Speech_C
  outputs [#13, #14]
  properties {
   softTaskIcon=&T_UI_NPC_Quest
   generalConfig {contextualMessages [
     {message=Can you help my client with this?, contextRequirements [], priority {value=0, curve {curveTable=null, rowName=None}}}
     {message=My client needs a hand. Or two., contextRequirements [], priority {value=0, curve {curveTable=null, rowName=None}}}
     {message=Help my client. Help me. Get Paid., contextRequirements [], priority {value=0, curve {curveTable=null, rowName=None}}}
     {
      message=At least my client's basement isn't full of Llamas again...
      contextRequirements []
      priority {value=0, curve {curveTable=null, rowName=None}}
     }
     {message=My client requested your services...., contextRequirements [], priority {value=0, curve {curveTable=null, rowName=None}}}
    ]}
  }
  sub_nodes [
   FortConversationRequirement_HasService_4 {class=FortConversationRequirement_HasService, properties {serviceTag {tagName=Tandem.Service.Quest2}}}
   FortConversationRequirement_AllSlottedQuestPrerequisitesCompleted_5 {class=FortConversationRequirement_AllSlottedQuestPrerequisitesCompleted, properties {questSlot=2}}
   FortConversationChoiceWithPreviewNode_4 {class=FortConversationChoiceWithPreviewNode, properties {defaultChoiceDisplayText=Quest}}
   FortConversationRequirement_SlottedQuestNotCompletedThisMatch_4 {class=FortConversationRequirement_SlottedQuestNotCompletedThisMatch, properties {questSlot=2}}
   FortConversationRequirement_HasNoActiveQuestsFromSlottedQuestProvider_4 {class=FortConversationRequirement_HasNoActiveQuestsFromSlottedQuestProvider, properties {questSlot=2}}
   FortConversationRequirement_ControllerRequirement_1 {
    class=FortConversationRequirement_ControllerRequirement
    properties {
     nodeName=Require: Has Recorded Conversation Event With NPC
     requirement {
      class=FortControllerRequirement_HasRecordedConversationEventWithNPC
      properties {bInverseCondition=true, queryMethod=CheckAgainstCurrentConversationParticipant, queryEventTag {tagName=Conversation.Event.GrantedQuest}}
     }
    }
   }
  ]
 }
 Back_C_9 {
  id=#14
  class=Back_C
  sub_nodes [ConversationChoiceNode_8 {class=ConversationChoiceNode, properties {defaultChoiceDisplayText="BACK!"}}]
 }
 GrantSlottedQuest_C_9 {
  id=#13
  class=GrantSlottedQuest_C
  outputs []
  properties {bNotifyQuestProvider=true, bShouldDisplayQuestProviderImage=true, bShouldOverrideSpeakerText=false, questSlot=2}
  sub_nodes [
   ConversationChoiceNode_9 {class=ConversationChoiceNode, properties {defaultChoiceDisplayText="ACCEPT!"}}
   SideEffect_PlayServiceUsedEffects_C_1 {class=SideEffect_PlayServiceUsedEffects_C}
  ]
 }
 Speech_C_8 {
  id=#2
  class=Speech_C
  outputs [#15, #16]
  properties {
   softTaskIcon=&T_UI_NPC_Quest
   generalConfig {contextualMessages [
     {message=Can you help my client with this?, contextRequirements [], priority {value=0, curve {curveTable=null, rowName=None}}}
     {message=My client needs a hand. Or two., contextRequirements [], priority {value=0, curve {curveTable=null, rowName=None}}}
     {message=Help my client. Help me. Get Paid., contextRequirements [], priority {value=0, curve {curveTable=null, rowName=None}}}
     {
      message=At least my client's basement isn't full of Llamas again...
      contextRequirements []
      priority {value=0, curve {curveTable=null, rowName=None}}
     }
     {message=My client requested your services...., contextRequirements [], priority {value=0, curve {curveTable=null, rowName=None}}}
    ]}
  }
  sub_nodes [
   FortConversationRequirement_HasService_3 {class=FortConversationRequirement_HasService, properties {serviceTag {tagName=Tandem.Service.Quest1}}}
   FortConversationRequirement_AllSlottedQuestPrerequisitesCompleted_3 {class=FortConversationRequirement_AllSlottedQuestPrerequisitesCompleted}
   FortConversationChoiceWithPreviewNode_3 {class=FortConversationChoiceWithPreviewNode, properties {defaultChoiceDisplayText=Quest}}
   FortConversationRequirement_SlottedQuestNotCompletedThisMatch_3 {class=FortConversationRequirement_SlottedQuestNotCompletedThisMatch}
   FortConversationRequirement_HasNoActiveQuestsFromSlottedQuestProvider_3 {class=FortConversationRequirement_HasNoActiveQuestsFromSlottedQuestProvider}
   FortConversationRequirement_ControllerRequirement_0 {
    class=FortConversationRequirement_ControllerRequirement
    properties {
     nodeName=Require: Has Recorded Conversation Event With NPC
     requirement {
      class=FortControllerRequirement_HasRecordedConversationEventWithNPC
      properties {bInverseCondition=true, queryMethod=CheckAgainstCurrentConversationParticipant, queryEventTag {tagName=Conversation.Event.GrantedQuest}}
     }
    }
   }
  ]
 }
 Back_C_8 {
  id=#16
  class=Back_C
  sub_nodes [ConversationChoiceNode_6 {class=ConversationChoiceNode, properties {defaultChoiceDisplayText="BACK!"}}]
 }
 GrantSlottedQuest_C_8 {
  id=#15
  class=GrantSlottedQuest_C
  outputs []
  properties {bNotifyQuestProvider=true, bShouldDisplayQuestProviderImage=true, bShouldOverrideSpeakerText=false}
  sub_nodes [
   ConversationChoiceNode_7 {class=ConversationChoiceNode, properties {defaultChoiceDisplayText="ACCEPT!"}}
   SideEffect_PlayServiceUsedEffects_C_0 {class=SideEffect_PlayServiceUsedEffects_C}
  ]
 }
]
