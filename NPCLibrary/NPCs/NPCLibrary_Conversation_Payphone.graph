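# Conversation database for the payphone NPC, the entry conversation where a player is offered work.
# The Payphone entry point leads to one ringing greeting Speech node, which swaps to a "nothing more today" line once this participant has a recorded GrantedQuest event with the player.
# The speech then branches through link nodes to separate payphone entry tags for Season 17 narrative, quests and Season 24, keeping that content outside this database.
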ConversationDatabase /NPCLibrary/NPCs/NPCLibrary_Conversation_Payphone
entry_points [{tag=Conversation.Entry.Payphone, destinations [#1]}]
nodes [
 ConversationEntryPointNode_7 {
  id=#1
  class=ConversationEntryPointNode
  outputs [#2]
  properties {entryTag {tagName=Conversation.Entry.Payphone}}
 }
 Speech_C_0 {
  id=#2
  class=Speech_C
  outputs [#3, #4, #5]
  properties {
   nodeName=""
   generalConfig {
    defaultMessage="<b> *RING RING* </> Hey... Looking for more work, huh?"
    contextualMessages [{
      message="Sorry! That's all I've got for you today, try somebody else!"
      requirementMatchPolicy=RequireAny
      contextRequirements [{requirement {
         class=FortControllerRequirement_HasRecordedConversationEventWithNPC
         properties {queryMethod=CheckAgainstCurrentConversationParticipant, queryEventTag {tagName=Conversation.Event.GrantedQuest}}
        }}]
      priority {curve {curveTable=null, rowName=None}}
     }]
   }
  }
 }
 ConversationLinkNode_2 {id=#5, class=ConversationLinkNode, properties {remoteEntryTag {tagName=Conversation.Entry.Season24.Payphone}}}
 ConversationLinkNode_1 {id=#4, class=ConversationLinkNode, properties {remoteEntryTag {tagName=Conversation.Entry.Payphone.Quests}}}
 ConversationLinkNode_0 {
  id=#3
  class=ConversationLinkNode
  properties {remoteEntryTag {tagName=Conversation.Entry.Season17.Narrative.Payphones}}
 }
]
