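# Conversation database for a printer NPC, entered through the Conversation.Entry.Printer tag.
# A speech node delivers a quip about spending bars, then branches to two player choices.
# The "OUT OF ORDER" choice runs a give-gift task and then a Repeat node, while BACK leads to a Back node.
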ConversationDatabase /NPCLibrary/NPCs/NPCLibrary_Conversation_Printer
entry_points [{tag=Conversation.Entry.Printer, destinations [#1]}]
nodes [
 ConversationEntryPointNode_7 {
  id=#1
  class=ConversationEntryPointNode
  outputs [#2]
  properties {entryTag {tagName=Conversation.Entry.Printer}}
 }
 Speech_C_4 {
  id=#2
  class=Speech_C
  outputs [#3, #4]
  properties {nodeName="", generalConfig {defaultMessage="Spend all those bars in one place!"}}
 }
 Back_C_5 {
  id=#4
  class=Back_C
  sub_nodes [ConversationChoiceNode_4 {class=ConversationChoiceNode, properties {defaultChoiceDisplayText=BACK}}]
 }
 FortConversationTaskNode_GiveGift_0 {
  id=#3
  class=FortConversationTaskNode_GiveGift
  outputs [#5]
  sub_nodes [ConversationChoiceNode_0 {class=ConversationChoiceNode, properties {defaultChoiceDisplayText=OUT OF ORDER}}]
 }
 Repeat_C_2 {id=#5, class=Repeat_C}
]
