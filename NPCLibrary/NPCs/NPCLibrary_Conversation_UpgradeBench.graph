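# Conversation database for the NPC upgrade bench service, entered through Conversation.Entry.UpgradeBench.
# A greeting speech leads to an UPGRADE preview choice, gated on the Tandem.Service.Upgrade service, whose line varies by NPC entry tag with a default fallback.
# That speech branches to BACK or to a CONFIRM choice that runs UpgradeItem with service-used effects, then goes back.
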ConversationDatabase /NPCLibrary/NPCs/NPCLibrary_Conversation_UpgradeBench
entry_points [{tag=Conversation.Entry.UpgradeBench, destinations [#1]}]
nodes [
 ConversationEntryPointNode_7 {
  id=#1
  class=ConversationEntryPointNode
  outputs [#2]
  properties {entryTag {tagName=Conversation.Entry.UpgradeBench}}
 }
 Speech_C_4 {
  id=#2
  class=Speech_C
  outputs [#3]
  properties {nodeName="", generalConfig {defaultMessage="Spend all those bars in one place!"}}
 }
 Speech_C_15 {
  id=#3
  class=Speech_C
  outputs [#4, #5]
  properties {
   bUsePerNPC_ContextualMessageLookup=true
   softTaskIcon=/Game/Athena/HUD/Interaction/T_UI_NPC_Upgrade
   generalConfig {defaultMessage="Don't mind the duct tape!"}
   speakerEntryTagToConfig {
    "(TagName=\"Conversation.Entry.Dummy\")" {defaultMessage="Oooh, I <b>love</> an upgrade. Don't you?"}
    "(TagName=\"Conversation.Entry.JonesyTheFirst\")" {defaultMessage=Using the standard kit is such a rookie move.}
    "(TagName=\"Conversation.Entry.Jules\")" {defaultMessage=We can make that way better. Just you wait and see.}
    "(TagName=\"Conversation.Entry.Sparkplug\")" {defaultMessage=Let's get you upgraded.}
    "(TagName=\"Conversation.Entry.TheReaper\")" {defaultMessage=A professional keeps their equipment sharp.}
   }
  }
  sub_nodes [
   FortConversationRequirement_HasService_1 {class=FortConversationRequirement_HasService, properties {serviceTag {tagName=Tandem.Service.Upgrade}}}
   FortConversationChoiceWithPreviewNode_5 {class=FortConversationChoiceWithPreviewNode, properties {defaultChoiceDisplayText=UPGRADE}}
  ]
 }
 UpgradeItem_C_1 {
  id=#5
  class=UpgradeItem_C
  outputs [#6]
  sub_nodes [
   ConversationChoiceNode_0 {class=ConversationChoiceNode, properties {defaultChoiceDisplayText=CONFIRM}}
   SideEffect_PlayServiceUsedEffects_C_0 {class=SideEffect_PlayServiceUsedEffects_C}
  ]
 }
 Back_C_1 {id=#6, class=Back_C}
 Back_C_0 {
  id=#4
  class=Back_C
  sub_nodes [ConversationChoiceNode_6 {class=ConversationChoiceNode, properties {defaultChoiceDisplayText=BACK}}]
 }
]
