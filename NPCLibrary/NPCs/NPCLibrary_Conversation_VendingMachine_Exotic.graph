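# Conversation database for the exotic vending machine, which pitches Ace's weapons stash and sells items from it.
# The entry greeting branches into five BUY choices, each gated by its Tandem.Service.Sell1–5 service requirement and using a shared external sales speech collection.
# Each choice offers BACK or a SellLoot step for its slot, which plays service-used effects.
# All sales converge on one data-driven effect task, then a Repeat node.
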ConversationDatabase /NPCLibrary/NPCs/NPCLibrary_Conversation_VendingMachine_Exotic
entry_points [{tag=Conversation.Entry.VendingMachine.Exotic, destinations [#1]}]
nodes [
 ConversationEntryPointNode_7 {
  id=#1
  class=ConversationEntryPointNode
  outputs [#2]
  properties {entryTag {tagName=Conversation.Entry.VendingMachine.Exotic}}
 }
 Speech_C_3 {
  id=#2
  class=Speech_C
  outputs [
   #3
   #4
   #5
   #6
   #7
  ]
  properties {
   bLookupViaServiceProviderTag=true
   nodeName=""
   generalConfig {contextualMessages [
     {
      message=I raid Ace's cache. You buy Ace's stuff from me. Deal?
      contextRequirements []
      priority {value=0, curve {curveTable=null, rowName=None}}
     }
     {
      message=You ever wanted to shop Ace's personal weapons stash?
      contextRequirements []
      priority {value=0, curve {curveTable=null, rowName=None}}
     }
    ]}
  }
 }
 Speech_C_4 {
  id=#7
  class=Speech_C
  outputs [#8, #9]
  properties {
   bLookupViaServiceProviderTag=true
   bUseInlineConfig=false
   externalContextualMessageCollection=&/DA_Speech_VendingMachine_Exotic_Sales
   softTaskIcon=/Game/Athena/HUD/Interaction/T_UI_NPC_Money
  }
  sub_nodes [
   FortConversationRequirement_HasService_18 {class=FortConversationRequirement_HasService, properties {serviceTag {tagName=Tandem.Service.Sell5}}}
   FortConversationChoiceWithPreviewNode_17 {class=FortConversationChoiceWithPreviewNode, properties {defaultChoiceDisplayText=BUY}}
  ]
 }
 SellLoot_C_1 {
  id=#9
  class=SellLoot_C
  outputs [#10]
  properties {sellSlot=5}
  sub_nodes [SideEffect_PlayServiceUsedEffects_C_19 {class=SideEffect_PlayServiceUsedEffects_C}]
 }
 FortConversationTaskNode_ApplyDataDrivenEffect_0 {id=#10, class=FortConversationTaskNode_ApplyDataDrivenEffect, outputs [#11]}
 Repeat_C_1 {id=#11, class=Repeat_C}
 Back_C_1 {
  id=#8
  class=Back_C
  sub_nodes [ConversationChoiceNode_16 {class=ConversationChoiceNode, properties {defaultChoiceDisplayText=BACK}}]
 }
 Speech_C_0 {
  id=#6
  class=Speech_C
  outputs [#12, #13]
  properties {
   bLookupViaServiceProviderTag=true
   bUseInlineConfig=false
   externalContextualMessageCollection=&DA_Speech_VendingMachine_Exotic_Sales
   softTaskIcon=&T_UI_NPC_Money
  }
  sub_nodes [
   FortConversationRequirement_HasService_17 {class=FortConversationRequirement_HasService, properties {serviceTag {tagName=Tandem.Service.Sell4}}}
   FortConversationChoiceWithPreviewNode_16 {class=FortConversationChoiceWithPreviewNode, properties {defaultChoiceDisplayText=BUY}}
  ]
 }
 SellLoot_C_0 {
  id=#13
  class=SellLoot_C
  outputs [#10]
  properties {sellSlot=4}
  sub_nodes [SideEffect_PlayServiceUsedEffects_C_18 {class=SideEffect_PlayServiceUsedEffects_C}]
 }
 Back_C_0 {
  id=#12
  class=Back_C
  sub_nodes [ConversationChoiceNode_15 {class=ConversationChoiceNode, properties {defaultChoiceDisplayText=BACK}}]
 }
 Speech_C_7 {
  id=#5
  class=Speech_C
  outputs [#14, #15]
  properties {
   bLookupViaServiceProviderTag=true
   bUseInlineConfig=false
   externalContextualMessageCollection=&DA_Speech_VendingMachine_Exotic_Sales
   softTaskIcon=&T_UI_NPC_Money
  }
  sub_nodes [
   FortConversationRequirement_HasService_16 {class=FortConversationRequirement_HasService, properties {serviceTag {tagName=Tandem.Service.Sell3}}}
   FortConversationChoiceWithPreviewNode_15 {class=FortConversationChoiceWithPreviewNode, properties {defaultChoiceDisplayText=BUY}}
  ]
 }
 SellLoot_C_4 {
  id=#15
  class=SellLoot_C
  outputs [#10]
  properties {sellSlot=3}
  sub_nodes [SideEffect_PlayServiceUsedEffects_C_17 {class=SideEffect_PlayServiceUsedEffects_C}]
 }
 Back_C_6 {
  id=#14
  class=Back_C
  sub_nodes [ConversationChoiceNode_14 {class=ConversationChoiceNode, properties {defaultChoiceDisplayText=BACK}}]
 }
 Speech_C_1 {
  id=#4
  class=Speech_C
  outputs [#16, #17]
  properties {
   bLookupViaServiceProviderTag=true
   bUseInlineConfig=false
   externalContextualMessageCollection=&DA_Speech_VendingMachine_Exotic_Sales
   softTaskIcon=&T_UI_NPC_Money
  }
  sub_nodes [
   FortConversationChoiceWithPreviewNode_13 {class=FortConversationChoiceWithPreviewNode, properties {defaultChoiceDisplayText=BUY}}
   FortConversationRequirement_HasService_14 {class=FortConversationRequirement_HasService, properties {serviceTag {tagName=Tandem.Service.Sell1}}}
  ]
 }
 SellLoot_C_2 {
  id=#17
  class=SellLoot_C
  outputs [#10]
  sub_nodes [SideEffect_PlayServiceUsedEffects_C_15 {class=SideEffect_PlayServiceUsedEffects_C}]
 }
 Back_C_2 {
  id=#16
  class=Back_C
  sub_nodes [ConversationChoiceNode_12 {class=ConversationChoiceNode, properties {defaultChoiceDisplayText=BACK}}]
 }
 Speech_C_2 {
  id=#3
  class=Speech_C
  outputs [#18, #19]
  properties {
   bLookupViaServiceProviderTag=true
   bUseInlineConfig=false
   externalContextualMessageCollection=&DA_Speech_VendingMachine_Exotic_Sales
   softTaskIcon=&T_UI_NPC_Money
  }
  sub_nodes [
   FortConversationRequirement_HasService_15 {class=FortConversationRequirement_HasService, properties {serviceTag {tagName=Tandem.Service.Sell2}}}
   FortConversationChoiceWithPreviewNode_14 {class=FortConversationChoiceWithPreviewNode, properties {defaultChoiceDisplayText=BUY}}
  ]
 }
 SellLoot_C_3 {
  id=#19
  class=SellLoot_C
  outputs [#10]
  properties {sellSlot=2}
  sub_nodes [SideEffect_PlayServiceUsedEffects_C_16 {class=SideEffect_PlayServiceUsedEffects_C}]
 }
 Back_C_5 {
  id=#18
  class=Back_C
  sub_nodes [ConversationChoiceNode_13 {class=ConversationChoiceNode, properties {defaultChoiceDisplayText=BACK}}]
 }
]
