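# Sound cue for buying from an NPC in the conversation UI, routed to the non-critical in-game UI sound class.
# A mixer layers three waves: a radial click with slight random pitch variation, a buy sound after a short fixed delay at reduced level, and a cash-buy music stinger.
# The whole cue plays at a lowered overall volume.
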SoundCue /Game/Athena/Sounds/NPCs/Conversation/Conversation_Buy_Cue
root=Mixer_0
nodes [
 {id=Delay_0, children [WavePlayer_1], properties {delayMax=0.3, delayMin=0.3}}
 {id=Mixer_0, children [Modulator_1,Delay_0,WavePlayer_2], properties {inputVolume [1,0.65,1]}}
 {id=Modulator_1, children [WavePlayer_3], properties {pitchMax=1.02, pitchMin=0.98, volumeMax=1, volumeMin=1}}
 {id=WavePlayer_1, properties {soundWaveAssetPtr=&/Conversation_Buy_01}}
 {id=WavePlayer_2, properties {soundWaveAssetPtr=&/NPC_Cash_Buy_MX_Stinger}}
 {id=WavePlayer_3, properties {soundWaveAssetPtr=&/NPC_Radial_Click}}
]
properties {maxDistance=2097152, soundClassObject=&Fort_UI_InGame_NonCritical, volumeMultiplier=0.7}
