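# Conversation database for the malfunctioning Stamina vending machine, offering a random "roll the dice" purchase.
# A glitch line leads to a BUY choice gated on the RollTheDice service, which offers BACK or ACCEPT.
# ACCEPT gives the player an NPC gift with purchase sound and service effects, then applies a data-driven effect and repeats.
# The BUY speech switches to a "clunk" message when the player has the just-purchased context tag.
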ConversationDatabase /NPCLibrary/NPCs/NPCLibrary_Conversation_VendingMachine_Malfunctioning_Stamina
entry_points [{tag=Conversation.Entry.VendingMachine.Malfunctioning.Stamina, destinations [#1]}]
nodes [
 ConversationEntryPointNode_7 {
  id=#1
  class=ConversationEntryPointNode
  outputs [#2]
  properties {entryTag {tagName=Conversation.Entry.VendingMachine.Malfunctioning.Stamina}}
 }
 Speech_C_1 {
  id=#2
  class=Speech_C
  outputs [#3]
  properties {
   nodeName=""
   generalConfig {
    defaultMessage=...
    contextualMessages [{message="<b>*glitch*</>", contextRequirements [], priority {value=0, curve {curveTable=null, rowName=None}}}]
   }
  }
 }
 Speech_C_2 {
  id=#3
  class=Speech_C
  outputs [#4, #5]
  properties {
   softTaskIcon=/Game/Athena/HUD/Interaction/T_UI_VendingMachineRandom
   generalConfig {
    defaultMessage=...
    contextualMessages [{
      message="<b>*clunk*</> "
      contextRequirements [{requirement {class=FortControllerRequirement_HasOwnedTag, properties {ownedTag {tagName=Conversation.Context.ServiceJustPurchased}}}}]
      priority {curve {curveTable=null, rowName=None}}
     }]
   }
  }
  sub_nodes [
   FortConversationChoiceWithPreviewNode_0 {class=FortConversationChoiceWithPreviewNode, properties {defaultChoiceDisplayText=BUY}}
   FortConversationRequirement_HasService_1 {class=FortConversationRequirement_HasService, properties {serviceTag {tagName=Tandem.Service.RollTheDice}}}
  ]
 }
 RollTheDice_C_1 {
  id=#5
  class=RollTheDice_C
  outputs [#6]
  properties {
   choiceSound=/Game/Athena/Sounds/NPCs/Conversation/Conversation_Buy_Cue
   effectRecipientConfigs [{recipient=Player, effects [{class=FortControllerEffect_NPCGiveGift}]}]
  }
  sub_nodes [
   ConversationChoiceNode_0 {class=ConversationChoiceNode, properties {defaultChoiceDisplayText=ACCEPT}}
   SideEffect_PlayServiceUsedEffects_C_1 {class=SideEffect_PlayServiceUsedEffects_C}
  ]
 }
 FortConversationTaskNode_ApplyDataDrivenEffect_2 {id=#6, class=FortConversationTaskNode_ApplyDataDrivenEffect, outputs [#7]}
 Repeat_C_0 {id=#7, class=Repeat_C}
 Back_C_1 {
  id=#4
  class=Back_C
  sub_nodes [ConversationChoiceNode_5 {class=ConversationChoiceNode, properties {defaultChoiceDisplayText=BACK}}]
 }
]
