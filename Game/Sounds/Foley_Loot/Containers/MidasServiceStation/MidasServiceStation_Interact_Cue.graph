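# Sound cue for interacting with the Midas Service Station loot container.
# A Random node picks between two equally weighted interact wave variations.
# It plays through the Fort_Loot_Reward_PSM sound class and overrides concurrency to cap simultaneous instances at three.
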SoundCue /Game/Sounds/Foley_Loot/Containers/MidasServiceStation/MidasServiceStation_Interact_Cue
root=Random_1
nodes [
 {id=Random_1, children [WavePlayer_0,WavePlayer_3], properties {weights [1,1]}}
 {id=WavePlayer_0, properties {soundWaveAssetPtr=&/MidasServiceStation_Interact_02}}
 {id=WavePlayer_3, properties {soundWaveAssetPtr=&/MidasServiceStation_Interact_01}}
]
properties {
 bOverrideConcurrency=true
 maxDistance=2097152
 soundClassObject=&Fort_Loot_Reward_PSM
 volumeMultiplier=1
 concurrencyOverrides {maxCount=3}
}
