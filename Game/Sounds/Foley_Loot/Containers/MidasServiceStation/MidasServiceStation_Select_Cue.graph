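# Sound cue for the select action of the Midas Service Station, a container in the loot foley sounds.
# A Random node picks between two select wave variations with equal weight.
# It uses the Fort_Loot_Reward_PSM sound class and overrides concurrency to limit simultaneous plays to three.
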SoundCue /Game/Sounds/Foley_Loot/Containers/MidasServiceStation/MidasServiceStation_Select_Cue
root=Random_0
nodes [
 {id=Random_0, children [WavePlayer_1,WavePlayer_2], properties {weights [1,1]}}
 {id=WavePlayer_1, properties {soundWaveAssetPtr=&/MidasServiceStation_Select_01}}
 {id=WavePlayer_2, properties {soundWaveAssetPtr=&/MidasServiceStation_Select_02}}
]
properties {
 bOverrideConcurrency=true
 maxDistance=2097152
 soundClassObject=&Fort_Loot_Reward_PSM
 volumeMultiplier=1
 concurrencyOverrides {maxCount=3}
}
