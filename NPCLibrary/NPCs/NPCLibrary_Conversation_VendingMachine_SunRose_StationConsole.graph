ConversationDatabase /NPCLibrary/NPCs/NPCLibrary_Conversation_VendingMachine_SunRose_StationConsole
entry_points [{tag=Conversation.Entry.VendingMachine.SunRose.Station, destinations [#1]}]
nodes [
 ConversationEntryPointNode_7 {
  id=#1
  class=ConversationEntryPointNode
  outputs [#2]
  properties {entryTag {tagName=Conversation.Entry.VendingMachine.SunRose.Station}}
 }
 Speech_C_3 {
  id=#2
  class=Speech_C
  outputs [
   #3
   #4
   #5
   #6
   #7
   #8
   #9
   #10
   #11
   #12
   #13
   #14
   #15
   #16
   #17
  ]
  properties {
   nodeName=""
   generalConfig {contextualMessages [
     {
      message="Greetings, user. All services are online and functional."
      contextRequirements []
      priority {value=0, curve {curveTable=null, rowName=None}}
     }
     {message="Greetings, user. How may I serve you today?", contextRequirements [], priority {value=0, curve {curveTable=null, rowName=None}}}
    ]}
  }
 }
 Speech_C_12 {
  id=#17
  class=Speech_C
  outputs [#18, #19]
  properties {
   bUsePerNPC_ContextualMessageLookup=true
   choiceSound=/Game/Sounds/Foley_Loot/Containers/MidasServiceStation/MidasServiceStation_Select_Cue
   softTaskIcon=/Game/Athena/HUD/Interaction/T_UI_NPC_Upgrade
   generalConfig {defaultMessage="Don't mind the duct tape!"}
   speakerEntryTagToConfig {
    "(TagName=\"Conversation.Entry.Dummy\")" {defaultMessage="Oooh, I <b>love</> an upgrade. Don't you?"}
    "(TagName=\"Conversation.Entry.FishThicc\")" {defaultMessage=Add some reps to your set.}
    "(TagName=\"Conversation.Entry.JonesyTheFirst\")" {defaultMessage=Using the standard kit is such a rookie move.}
    "(TagName=\"Conversation.Entry.Jules\")" {defaultMessage=We can make that way better. Just you wait and see.}
    "(TagName=\"Conversation.Entry.MunitionsExpert\")" {defaultMessage=This will be useful to you.}
    "(TagName=\"Conversation.Entry.Sparkplug\")" {defaultMessage=Let's get you upgraded.}
    "(TagName=\"Conversation.Entry.TheReaper\")" {defaultMessage=A professional keeps their equipment sharp.}
   }
  }
  sub_nodes [
   FortConversationRequirement_HasService_12 {class=FortConversationRequirement_HasService, properties {serviceTag {tagName=Tandem.Service.GoldenUpgrade}}}
   FortConversationChoiceWithPreviewNode_14 {
    class=FortConversationChoiceWithPreviewNode
    properties {defaultChoiceDisplayText=UPGRADE, choiceTags {gameplayTags [{tagName=Conversation.Choice.CustomSortPriority.Low}]}}
   }
  ]
 }
 GoldenUpgrade_C_1 {
  id=#19
  class=GoldenUpgrade_C
  outputs [#20]
  properties {
   choiceSound=/Game/Sounds/Foley_Loot/Containers/MidasServiceStation/MidasServiceStation_Confirm_Cue
   effectRecipientConfigs [{
     recipient=Player
     effects [
      {class=FortControllerEffect_ApplyWeaponWrap, properties {wrap=/Game/Athena/Items/Cosmetics/ItemWraps/Wrap_CatBurglarGameplay}}
      {class=FortControllerEffect_UpgradeItem, properties {bUpgradeByOneTier=false, upgradeToRarity=Legendary}}
     ]
    }]
   requirements [{
     cannotUseReasonParameter=Cannot Upgrade to Legendary
     failureNodeBehaviour=FailedButVisible
     participantId {tagName=Conversation.Participant.Player}
     requirement {class=FortControllerRequirement_CanUpgradeWeapon, properties {bUpgradableToNextTier=false}}
    }]
  }
  sub_nodes [ConversationChoiceNode_9 {class=ConversationChoiceNode, properties {defaultChoiceDisplayText=CONFIRM}}]
 }
 Speech_C_1 {
  id=#20
  class=Speech_C
  outputs []
  properties {
   generalConfig {contextualMessages [
     {message="Thank you for your patronage!", contextRequirements [], priority {value=0, curve {curveTable=null, rowName=None}}}
     {message="At your service!", contextRequirements [], priority {value=0, curve {curveTable=null, rowName=None}}}
     {message="Happy to serve!", contextRequirements [], priority {value=0, curve {curveTable=null, rowName=None}}}
    ]}
   listenerTag=~
   speakerTag=~
  }
 }
 Back_C_13 {
  id=#18
  class=Back_C
  properties {choiceSound=/Game/Sounds/Foley_Loot/Containers/MidasServiceStation/MidasServiceStation_Interact_Cue}
  sub_nodes [ConversationChoiceNode_7 {class=ConversationChoiceNode, properties {defaultChoiceDisplayText=BACK}}]
 }
 Speech_C_16 {
  id=#16
  class=Speech_C
  outputs [#21, #22]
  properties {bUsePerNPC_ContextualMessageLookup=true, softTaskIcon=/Game/Athena/HUD/Interaction/T_UI_TipSmall, generalConfig {defaultMessage=...}}
  sub_nodes [
   FortConversationChoiceWithPreviewNode_11 {class=FortConversationChoiceWithPreviewNode, properties {defaultChoiceDisplayText="TIP BUS DRIVER (SMALL)"}}
   FortConversationRequirement_HasService_16 {class=FortConversationRequirement_HasService, properties {serviceTag {tagName=Tandem.Service.TipBusDriverSmall}}}
  ]
 }
 TipBusDriverSmall_C_2 {
  id=#22
  class=TipBusDriverSmall_C
  outputs [#20]
  properties {effectRecipientConfigs [{
     recipient=Player
     effects [{class=FortControllerEffect_TriggerCustomFeedMessage, properties {customFeedMessageTag {tagName=Tandem.Service.TipBusDriverSmall}}}]
    }]}
  sub_nodes [ConversationChoiceNode_29 {class=ConversationChoiceNode, properties {defaultChoiceDisplayText=ACCEPT}}]
 }
 Back_C_17 {
  id=#21
  class=Back_C
  sub_nodes [ConversationChoiceNode_31 {class=ConversationChoiceNode, properties {defaultChoiceDisplayText=BACK}}]
 }
 Speech_C_17 {
  id=#15
  class=Speech_C
  outputs [#23, #24]
  properties {bUsePerNPC_ContextualMessageLookup=true, softTaskIcon=/Game/Athena/HUD/Interaction/T_UI_TipBig, generalConfig {defaultMessage=...}}
  sub_nodes [
   FortConversationChoiceWithPreviewNode_12 {class=FortConversationChoiceWithPreviewNode, properties {defaultChoiceDisplayText="TIP BUS DRIVER (LARGE)"}}
   FortConversationRequirement_HasService_17 {class=FortConversationRequirement_HasService, properties {serviceTag {tagName=Tandem.Service.TipBusDriverLarge}}}
  ]
 }
 TipBusDriverLarge_C_1 {
  id=#24
  class=TipBusDriverLarge_C
  outputs [#20]
  properties {effectRecipientConfigs [{
     recipient=Player
     effects [{class=FortControllerEffect_TriggerCustomFeedMessage, properties {customFeedMessageTag {tagName=Tandem.Service.TipBusDriverLarge}}}]
    }]}
  sub_nodes [ConversationChoiceNode_28 {class=ConversationChoiceNode, properties {defaultChoiceDisplayText=ACCEPT}}]
 }
 Back_C_18 {
  id=#23
  class=Back_C
  sub_nodes [ConversationChoiceNode_39 {class=ConversationChoiceNode, properties {defaultChoiceDisplayText=BACK}}]
 }
 Speech_C_14 {
  id=#14
  class=Speech_C
  outputs [#25, #26]
  properties {bUsePerNPC_ContextualMessageLookup=true, softTaskIcon=/Game/UI/Icons/Icon-RandomDice/Icon-RandomDice-128, generalConfig {defaultMessage=...}}
  sub_nodes [
   FortConversationRequirement_HasService_11 {class=FortConversationRequirement_HasService, properties {serviceTag {tagName=Tandem.Service.RollTheDice}}}
   FortConversationChoiceWithPreviewNode_10 {class=FortConversationChoiceWithPreviewNode, properties {defaultChoiceDisplayText=ROLL THE DICE}}
  ]
 }
 RollTheDice_C_0 {
  id=#26
  class=RollTheDice_C
  outputs [#20]
  properties {effectRecipientConfigs [{recipient=Player, effects [{class=FortControllerEffect_NPCGiveGift}]}]}
  sub_nodes [ConversationChoiceNode_19 {class=ConversationChoiceNode, properties {defaultChoiceDisplayText=ACCEPT}}]
 }
 Back_C_15 {
  id=#25
  class=Back_C
  sub_nodes [ConversationChoiceNode_36 {class=ConversationChoiceNode, properties {defaultChoiceDisplayText=BACK}}]
 }
 Speech_C_10 {
  id=#13
  class=Speech_C
  outputs [#27, #28]
  properties {bUsePerNPC_ContextualMessageLookup=true, softTaskIcon=/Game/Athena/HUD/Interaction/T_UI_MassSupply, generalConfig {defaultMessage=...}}
  sub_nodes [
   FortConversationChoiceWithPreviewNode_20 {class=FortConversationChoiceWithPreviewNode, properties {defaultChoiceDisplayText=SupplyDrop}}
   FortConversationRequirement_HasService_10 {class=FortConversationRequirement_HasService, properties {serviceTag {tagName=Tandem.Service.SupplyDrop}}}
  ]
 }
 SupplyDrop_C_0 {
  id=#28
  class=SupplyDrop_C
  outputs [#20]
  sub_nodes [ConversationChoiceNode_18 {class=ConversationChoiceNode, properties {defaultChoiceDisplayText=ACCEPT}}]
 }
 Back_C_10 {
  id=#27
  class=Back_C
  sub_nodes [ConversationChoiceNode_37 {class=ConversationChoiceNode, properties {defaultChoiceDisplayText=BACK}}]
 }
 Speech_C_9 {
  id=#12
  class=Speech_C
  outputs [#29, #30]
  properties {bUsePerNPC_ContextualMessageLookup=true, softTaskIcon=/Game/Athena/HUD/Interaction/T_UI_NPC_Bounty, generalConfig {defaultMessage=...}}
  sub_nodes [
   FortConversationRequirement_HasService_9 {class=FortConversationRequirement_HasService, properties {serviceTag {tagName=Tandem.Service.PlayerBounty}}}
   FortConversationChoiceWithPreviewNode_9 {class=FortConversationChoiceWithPreviewNode, properties {defaultChoiceDisplayText=BOUNTY}}
  ]
 }
 GrantPlayerBounty_C_0 {
  id=#30
  class=GrantPlayerBounty_C
  outputs [#20]
  sub_nodes [
   ConversationChoiceNode_13 {class=ConversationChoiceNode, properties {defaultChoiceDisplayText=CONFIRM}}
   SideEffect_Client_PlayVOEvent_C_0 {class=SideEffect_Client_PlayVOEvent_C, properties {voiceOverTag {tagName=Pawn.Athena.NPC.Sound.Voice.Quest.Excited}}}
  ]
 }
 Back_C_9 {
  id=#29
  class=Back_C
  sub_nodes [
   ConversationChoiceNode_16 {class=ConversationChoiceNode, properties {defaultChoiceDisplayText=BACK}}
   SideEffect_Client_PlayVOEvent_C_1 {class=SideEffect_Client_PlayVOEvent_C, properties {voiceOverTag {tagName=Pawn.Athena.NPC.Sound.Voice.Quest.Disappointed}}}
  ]
 }
 Speech_C_8 {
  id=#11
  class=Speech_C
  outputs [#31, #32]
  properties {
   bUsePerNPC_ContextualMessageLookup=true
   softTaskIcon=/Game/Athena/HUD/Interaction/T_UI_NPC_Intel_NextStorm
   generalConfig {defaultMessage=Gathering Storm analytics...}
  }
  sub_nodes [
   FortConversationRequirement_HasService_5 {class=FortConversationRequirement_HasService, properties {serviceTag {tagName=Tandem.Service.ShowFutureStormCircle}}}
   FortConversationChoiceWithPreviewNode_6 {class=FortConversationChoiceWithPreviewNode, properties {defaultChoiceDisplayText=STORM FORECAST}}
  ]
 }
 ShowFutureStormCircle_C_0 {
  id=#32
  class=ShowFutureStormCircle_C
  outputs [#20]
  sub_nodes [ConversationChoiceNode_15 {class=ConversationChoiceNode, properties {defaultChoiceDisplayText=CONFIRM}}]
 }
 Back_C_8 {
  id=#31
  class=Back_C
  sub_nodes [ConversationChoiceNode_11 {class=ConversationChoiceNode, properties {defaultChoiceDisplayText=BACK}}]
 }
 Speech_C_7 {
  id=#10
  class=Speech_C
  outputs [#33, #34]
  properties {
   bUsePerNPC_ContextualMessageLookup=true
   choiceSound=&MidasServiceStation_Select_Cue
   softTaskIcon=/Game/Athena/HUD/Interaction/T_UI_NPC_Rift
   generalConfig {defaultMessage=Rift technology: online.}
  }
  sub_nodes [
   FortConversationRequirement_HasService_7 {class=FortConversationRequirement_HasService, properties {serviceTag {tagName=Tandem.Service.RiftToGo}}}
   FortConversationChoiceWithPreviewNode_7 {class=FortConversationChoiceWithPreviewNode, properties {defaultChoiceDisplayText=RIFT-TO-GO}}
  ]
 }
 RiftToGo_C_2 {
  id=#34
  class=RiftToGo_C
  outputs [#20]
  properties {
   choiceSound=&MidasServiceStation_Confirm_Cue
   softTaskIcon=&T_UI_NPC_Rift
  }
  sub_nodes [ConversationChoiceNode_17 {class=ConversationChoiceNode, properties {defaultChoiceDisplayText=CONFIRM}}]
 }
 Back_C_5 {
  id=#33
  class=Back_C
  properties {choiceSound=&MidasServiceStation_Interact_Cue}
  sub_nodes [ConversationChoiceNode_4 {class=ConversationChoiceNode, properties {defaultChoiceDisplayText=BACK}}]
 }
 Speech_C_15 {
  id=#9
  class=Speech_C
  outputs [#35, #36]
  properties {
   bUsePerNPC_ContextualMessageLookup=true
   choiceSound=&MidasServiceStation_Select_Cue
   softTaskIcon=&T_UI_NPC_Upgrade
   generalConfig {defaultMessage=...}
  }
  sub_nodes [
   FortConversationRequirement_HasService_4 {class=FortConversationRequirement_HasService, properties {serviceTag {tagName=Tandem.Service.Upgrade}}}
   FortConversationChoiceWithPreviewNode_5 {class=FortConversationChoiceWithPreviewNode, properties {defaultChoiceDisplayText=UPGRADE}}
  ]
 }
 UpgradeItem_C_0 {
  id=#36
  class=UpgradeItem_C
  outputs [#20]
  properties {choiceSound=&MidasServiceStation_Confirm_Cue}
  sub_nodes [ConversationChoiceNode_8 {class=ConversationChoiceNode, properties {defaultChoiceDisplayText=CONFIRM}}]
 }
 Back_C_2 {
  id=#35
  class=Back_C
  properties {choiceSound=&MidasServiceStation_Interact_Cue}
  sub_nodes [ConversationChoiceNode_6 {class=ConversationChoiceNode, properties {defaultChoiceDisplayText=BACK}}]
 }
 Speech_C_2 {
  id=#8
  class=Speech_C
  outputs [#37, #38]
  properties {
   bUsePerNPC_ContextualMessageLookup=true
   softTaskIcon=/Game/Athena/HUD/Interaction/T_UI_NPC_Intel
   generalConfig {defaultMessage=Onboard Prop-ifier: Active.}
  }
  sub_nodes [
   FortConversationRequirement_HasService_8 {class=FortConversationRequirement_HasService, properties {serviceTag {tagName=Tandem.Service.PropDisguise}}}
   FortConversationChoiceWithPreviewNode_8 {class=FortConversationChoiceWithPreviewNode, properties {defaultChoiceDisplayText=PROP DISGUISE}}
  ]
 }
 PropDisguise_C_0 {
  id=#38
  class=PropDisguise_C
  outputs [#20]
  sub_nodes [ConversationChoiceNode_14 {class=ConversationChoiceNode, properties {defaultChoiceDisplayText=CONFIRM}}]
 }
 Back_C_6 {
  id=#37
  class=Back_C
  sub_nodes [ConversationChoiceNode_12 {class=ConversationChoiceNode, properties {defaultChoiceDisplayText=BACK}}]
 }
 Speech_C_0 {
  id=#7
  class=Speech_C
  outputs [#39, #40]
  properties {softTaskIcon=/Game/Athena/HUD/Interaction/T_UI_NPC_Money, generalConfig {defaultMessage=...}}
  sub_nodes [
   FortConversationRequirement_HasService_0 {class=FortConversationRequirement_HasService, properties {serviceTag {tagName=Tandem.Service.Sell4}}}
   FortConversationChoiceWithPreviewNode_0 {class=FortConversationChoiceWithPreviewNode, properties {defaultChoiceDisplayText=BUY}}
  ]
 }
 SellLoot_C_2 {
  id=#40
  class=SellLoot_C
  outputs [#41]
  properties {sellSlot=4}
  sub_nodes [SideEffect_PlayServiceUsedEffects_C_3 {class=SideEffect_PlayServiceUsedEffects_C}]
 }
 FortConversationTaskNode_ApplyDataDrivenEffect_0 {id=#41, class=FortConversationTaskNode_ApplyDataDrivenEffect, outputs [#42]}
 Repeat_C_0 {id=#42, class=Repeat_C}
 Back_C_1 {
  id=#39
  class=Back_C
  sub_nodes [ConversationChoiceNode_0 {class=ConversationChoiceNode, properties {defaultChoiceDisplayText=BACK}}]
 }
 Speech_C_18 {
  id=#6
  class=Speech_C
  outputs [#43, #44]
  properties {
   bUsePerNPC_ContextualMessageLookup=true
   choiceSound=&MidasServiceStation_Select_Cue
   softTaskIcon=/Game/Athena/HUD/Interaction/T_UI_NPC_medic
   generalConfig {defaultMessage="Med-Mist reservoir: full, but overdue for descaling. "}
  }
  sub_nodes [
   FortConversationChoiceWithPreviewNode_13 {class=FortConversationChoiceWithPreviewNode, properties {defaultChoiceDisplayText=PATCH UP}}
   FortConversationRequirement_HasService_18 {class=FortConversationRequirement_HasService, properties {serviceTag {tagName=Tandem.Service.PatchUp}}}
  ]
 }
 PatchUp_C_2 {
  id=#44
  class=PatchUp_C
  outputs [#20]
  properties {
   choiceSound=&MidasServiceStation_Confirm_Cue
   effectRecipientConfigs [{
     recipient=Player
     effects [{
       class=FortControllerEffect_HealPlayer
       properties {
        gameplayEffect=/NPCLibrary/Conversation/GameplayEffects/GE_CBGA_NPCService_PatchUp.GE_CBGA_NPCService_PatchUp_C
        healthRestoreQuantity {value=1, curve {curveTable=/Game/Athena/Balance/DataTables/NPCServiceSettings, rowName=PatchUp.HealthRestoreQuantity}}
        parameterName {tagName=SetByCaller.Magnitude}
       }
      }]
    }]
   requirements [{
     cannotUseReasonParameter=Healthy
     failureNodeBehaviour=FailedButVisible
     participantId {tagName=Conversation.Participant.Player}
     requirement {
      class=FortControllerRequirement_AttributeThresholdCheck
      properties {
       queryAttribute {attribute=/Script/FortniteGame.FortHealthSet:Health, attributeName=Health, attributeOwner=&FortHealthSet}
       thresholdTestConfig {quantity {value=100, curve {rowName=Tandem.Gift.Health.Requirements.HealthThreshold}}}
      }
     }
    }]
  }
  sub_nodes [ConversationChoiceNode_5 {class=ConversationChoiceNode, properties {defaultChoiceDisplayText=CONFIRM}}]
 }
 Back_C_19 {
  id=#43
  class=Back_C
  properties {choiceSound=&MidasServiceStation_Interact_Cue}
  sub_nodes [ConversationChoiceNode_40 {class=ConversationChoiceNode, properties {defaultChoiceDisplayText=BACK}}]
 }
 Speech_C_6 {
  id=#5
  class=Speech_C
  outputs [#45, #46]
  properties {softTaskIcon=&T_UI_NPC_Money, generalConfig {defaultMessage=..}}
  sub_nodes [
   FortConversationRequirement_HasService_2 {class=FortConversationRequirement_HasService, properties {serviceTag {tagName=Tandem.Service.Sell3}}}
   FortConversationChoiceWithPreviewNode_1 {class=FortConversationChoiceWithPreviewNode, properties {defaultChoiceDisplayText=BUY}}
  ]
 }
 SellLoot_C_5 {
  id=#46
  class=SellLoot_C
  outputs [#41]
  properties {sellSlot=3}
  sub_nodes [SideEffect_PlayServiceUsedEffects_C_4 {class=SideEffect_PlayServiceUsedEffects_C}]
 }
 Back_C_4 {
  id=#45
  class=Back_C
  sub_nodes [ConversationChoiceNode_3 {class=ConversationChoiceNode, properties {defaultChoiceDisplayText=BACK}}]
 }
 Speech_C_5 {
  id=#4
  class=Speech_C
  outputs [#47, #48]
  properties {softTaskIcon=&T_UI_NPC_Money, generalConfig {defaultMessage=...}}
  sub_nodes [
   FortConversationRequirement_HasService_3 {class=FortConversationRequirement_HasService, properties {serviceTag {tagName=Tandem.Service.Sell2}}}
   FortConversationChoiceWithPreviewNode_4 {class=FortConversationChoiceWithPreviewNode, properties {defaultChoiceDisplayText=BUY}}
  ]
 }
 SellLoot_C_1 {
  id=#48
  class=SellLoot_C
  outputs [#41]
  properties {sellSlot=2}
  sub_nodes [SideEffect_PlayServiceUsedEffects_C_1 {class=SideEffect_PlayServiceUsedEffects_C}]
 }
 Back_C_3 {
  id=#47
  class=Back_C
  sub_nodes [ConversationChoiceNode_2 {class=ConversationChoiceNode, properties {defaultChoiceDisplayText=BACK}}]
 }
 Speech_C_4 {
  id=#3
  class=Speech_C
  outputs [#49, #50]
  properties {softTaskIcon=&T_UI_NPC_Money, generalConfig {defaultMessage=...}}
  sub_nodes [
   FortConversationRequirement_HasService_1 {class=FortConversationRequirement_HasService, properties {serviceTag {tagName=Tandem.Service.Sell1}}}
   FortConversationChoiceWithPreviewNode_3 {class=FortConversationChoiceWithPreviewNode, properties {defaultChoiceDisplayText=BUY}}
  ]
 }
 SellLoot_C_0 {
  id=#50
  class=SellLoot_C
  outputs [#41]
  sub_nodes [SideEffect_PlayServiceUsedEffects_C_0 {class=SideEffect_PlayServiceUsedEffects_C}]
 }
 Back_C_0 {
  id=#49
  class=Back_C
  sub_nodes [ConversationChoiceNode_1 {class=ConversationChoiceNode, properties {defaultChoiceDisplayText=BACK}}]
 }
]
