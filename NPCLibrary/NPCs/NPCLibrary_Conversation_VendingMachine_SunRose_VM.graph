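# Conversation database for the SunRose vending machine, offering up to four items for sale plus a patch-up heal.
# A greeting branches into choices gated by service tags, each leading to a confirm or Back step.
# All sales play service-used effects, then share one data-driven effect node that leads to Repeat.
# Patch Up heals by a curve-table amount, stays visible when its health-threshold check fails, and ends on a spoken line.
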ConversationDatabase /NPCLibrary/NPCs/NPCLibrary_Conversation_VendingMachine_SunRose_VM
entry_points [{tag=Conversation.Entry.VendingMachine.SunRose.VM, destinations [#1]}]
nodes [
 ConversationEntryPointNode_7 {
  id=#1
  class=ConversationEntryPointNode
  outputs [#2]
  properties {entryTag {tagName=Conversation.Entry.VendingMachine.SunRose.VM}}
 }
 Speech_C_3 {
  id=#2
  class=Speech_C
  outputs [
   #3
   #4
   #5
   #6
   #7
  ]
  properties {nodeName="", generalConfig {defaultMessage=From the vault of the Golden King...}}
 }
 Speech_C_0 {
  id=#7
  class=Speech_C
  outputs [#8, #9]
  properties {softTaskIcon=/Game/Athena/HUD/Interaction/T_UI_NPC_Money, generalConfig {defaultMessage=...}}
  sub_nodes [
   FortConversationRequirement_HasService_0 {class=FortConversationRequirement_HasService, properties {serviceTag {tagName=Tandem.Service.Sell4}}}
   FortConversationChoiceWithPreviewNode_0 {class=FortConversationChoiceWithPreviewNode, properties {defaultChoiceDisplayText=BUY}}
  ]
 }
 SellLoot_C_2 {
  id=#9
  class=SellLoot_C
  outputs [#10]
  properties {sellSlot=4}
  sub_nodes [SideEffect_PlayServiceUsedEffects_C_3 {class=SideEffect_PlayServiceUsedEffects_C}]
 }
 FortConversationTaskNode_ApplyDataDrivenEffect_0 {id=#10, class=FortConversationTaskNode_ApplyDataDrivenEffect, outputs [#11]}
 Repeat_C_0 {id=#11, class=Repeat_C}
 Back_C_1 {
  id=#8
  class=Back_C
  sub_nodes [ConversationChoiceNode_0 {class=ConversationChoiceNode, properties {defaultChoiceDisplayText=BACK}}]
 }
 Speech_C_18 {
  id=#6
  class=Speech_C
  outputs [#12, #13]
  properties {bUsePerNPC_ContextualMessageLookup=true, softTaskIcon=/Game/Athena/HUD/Interaction/T_UI_NPC_medic}
  sub_nodes [
   FortConversationChoiceWithPreviewNode_13 {class=FortConversationChoiceWithPreviewNode, properties {defaultChoiceDisplayText=PATCH UP}}
   FortConversationRequirement_HasService_18 {class=FortConversationRequirement_HasService, properties {serviceTag {tagName=Tandem.Service.PatchUp}}}
  ]
 }
 PatchUp_C_2 {
  id=#13
  class=PatchUp_C
  outputs [#14]
  properties {
   effectRecipientConfigs [{
     recipient=Player
     effects [{
       class=FortControllerEffect_HealPlayer
       properties {
        gameplayEffect=/NPCLibrary/Conversation/GameplayEffects/GE_CBGA_NPCService_PatchUp.GE_CBGA_NPCService_PatchUp_C
        healthRestoreQuantity {value=1, curve {curveTable=/Game/Athena/Balance/DataTables/NPCServiceSettings, rowName=PatchUp.HealthRestoreQuantity}}
        parameterName {tagName=SetByCaller.Magnitude}
       }
      }]
    }]
   requirements [{
     cannotUseReasonParameter=Healthy
     failureNodeBehaviour=FailedButVisible
     participantId {tagName=Conversation.Participant.Player}
     requirement {
      class=FortControllerRequirement_AttributeThresholdCheck
      properties {
       queryAttribute {attribute=/Script/FortniteGame.FortHealthSet:Health, attributeName=Health, attributeOwner=&FortHealthSet}
       thresholdTestConfig {quantity {value=100, curve {rowName=Tandem.Gift.Health.Requirements.HealthThreshold}}}
      }
     }
    }]
  }
  sub_nodes [ConversationChoiceNode_5 {class=ConversationChoiceNode, properties {defaultChoiceDisplayText=ACCEPT}}]
 }
 Speech_C_1 {
  id=#14
  class=Speech_C
  outputs []
  properties {generalConfig {defaultMessage=Need anything else? A juicebox? Some crackers? A hug?, defaultMetadataParameters [Style {value=Shock}]}}
 }
 Back_C_19 {
  id=#12
  class=Back_C
  sub_nodes [ConversationChoiceNode_40 {class=ConversationChoiceNode, properties {defaultChoiceDisplayText=BACK}}]
 }
 Speech_C_6 {
  id=#5
  class=Speech_C
  outputs [#15, #16]
  properties {softTaskIcon=&T_UI_NPC_Money, generalConfig {defaultMessage=...}}
  sub_nodes [
   FortConversationRequirement_HasService_2 {class=FortConversationRequirement_HasService, properties {serviceTag {tagName=Tandem.Service.Sell3}}}
   FortConversationChoiceWithPreviewNode_1 {class=FortConversationChoiceWithPreviewNode, properties {defaultChoiceDisplayText=BUY}}
  ]
 }
 SellLoot_C_5 {
  id=#16
  class=SellLoot_C
  outputs [#10]
  properties {sellSlot=3}
  sub_nodes [SideEffect_PlayServiceUsedEffects_C_4 {class=SideEffect_PlayServiceUsedEffects_C}]
 }
 Back_C_4 {
  id=#15
  class=Back_C
  sub_nodes [ConversationChoiceNode_3 {class=ConversationChoiceNode, properties {defaultChoiceDisplayText=BACK}}]
 }
 Speech_C_5 {
  id=#4
  class=Speech_C
  outputs [#17, #18]
  properties {softTaskIcon=&T_UI_NPC_Money, generalConfig {defaultMessage=...}}
  sub_nodes [
   FortConversationRequirement_HasService_3 {class=FortConversationRequirement_HasService, properties {serviceTag {tagName=Tandem.Service.Sell2}}}
   FortConversationChoiceWithPreviewNode_4 {class=FortConversationChoiceWithPreviewNode, properties {defaultChoiceDisplayText=BUY}}
  ]
 }
 SellLoot_C_1 {
  id=#18
  class=SellLoot_C
  outputs [#10]
  properties {sellSlot=2}
  sub_nodes [SideEffect_PlayServiceUsedEffects_C_1 {class=SideEffect_PlayServiceUsedEffects_C}]
 }
 Back_C_3 {
  id=#17
  class=Back_C
  sub_nodes [ConversationChoiceNode_2 {class=ConversationChoiceNode, properties {defaultChoiceDisplayText=BACK}}]
 }
 Speech_C_4 {
  id=#3
  class=Speech_C
  outputs [#19, #20]
  properties {softTaskIcon=&T_UI_NPC_Money, generalConfig {defaultMessage=...}}
  sub_nodes [
   FortConversationRequirement_HasService_1 {class=FortConversationRequirement_HasService, properties {serviceTag {tagName=Tandem.Service.Sell1}}}
   FortConversationChoiceWithPreviewNode_3 {class=FortConversationChoiceWithPreviewNode, properties {defaultChoiceDisplayText=BUY}}
  ]
 }
 SellLoot_C_0 {
  id=#20
  class=SellLoot_C
  outputs [#10]
  sub_nodes [SideEffect_PlayServiceUsedEffects_C_0 {class=SideEffect_PlayServiceUsedEffects_C}]
 }
 Back_C_0 {
  id=#19
  class=Back_C
  sub_nodes [ConversationChoiceNode_1 {class=ConversationChoiceNode, properties {defaultChoiceDisplayText=BACK}}]
 }
]
