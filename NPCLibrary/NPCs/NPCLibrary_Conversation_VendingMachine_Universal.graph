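# Universal weapon vending machine conversation that greets the player and sells guns, ammo and explosives in a comically trusting "Cold Blooded" voice.
# A greeting branches into five category menus, shown only when the machine offers the matching sell-subcategory service.
# Each category has service-gated item slots with previewed BUY choices that lead to SellLoot, a data-driven effect and a repeat.
# A just-purchased context tag unlocks alternate post-sale lines.
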ConversationDatabase /NPCLibrary/NPCs/NPCLibrary_Conversation_VendingMachine_Universal
entry_points [{tag=Conversation.Entry.VendingMachine.Universal, destinations [#1]}]
nodes [
 ConversationEntryPointNode_7 {
  id=#1
  class=ConversationEntryPointNode
  outputs [#2]
  properties {entryTag {tagName=Conversation.Entry.VendingMachine.Universal}}
 }
 Speech_C_3 {
  id=#2
  class=Speech_C
  outputs [
   #3
   #4
   #5
   #6
   #7
  ]
  properties {
   nodeName=""
   generalConfig {
    defaultMessage="GREETINGS, [COLDBLOODED ID NUMBER = 00000]!"
    contextualMessages [
     {message="ARMORY ACCESS: GRANTED, FOR SOME REASON.", contextRequirements [], priority {value=0, curve {curveTable=null, rowName=None}}}
     {message="GREETINGS, [COLDBLOODED ID NUMBER = 00000]!", contextRequirements [], priority {value=0, curve {curveTable=null, rowName=None}}}
    ]
   }
  }
 }
 Speech_C_2 {
  id=#7
  class=Speech_C
  outputs [
   #8
   #9
   #10
   #11
   #12
  ]
  properties {
   softTaskIcon=/Game/UI/Foundation/Textures/Icons/ItemTypes/T-Icon-Pistol-128
   generalConfig {contextualMessages [
     {message="MAKE A SELECTION, TRUSTED USER!", contextRequirements [], priority {value=0, curve {curveTable=null, rowName=None}}}
     {message="HOW CAN I ARM YOU TODAY, HONORABLE PATRON?", contextRequirements [], priority {value=0, curve {curveTable=null, rowName=None}}}
    ]}
  }
  sub_nodes [
   ConversationChoiceNode_6 {
    class=ConversationChoiceNode
    properties {defaultChoiceDisplayText="Light Guns && Ammo", choiceTags {gameplayTags [{tagName=Conversation.Choice.CustomSortPriority.Highest}]}}
   }
   FortConversationRequirement_HasService_0 {class=FortConversationRequirement_HasService, properties {bRequireExactMatch=false, serviceTag {tagName=Tandem.Service.SellSubCategory1}}}
  ]
 }
 ReturnToStart_C_1 {
  id=#12
  class=ReturnToStart_C
  properties {
   choiceSound=/Game/Athena/Sounds/NPCs/Conversation/Conversation_Back_Cue
   hoverChoiceSound=/Game/Athena/Sounds/NPCs/Conversation/Conversation_Hover_Back_Cue
   softTaskIcon=&T_UI_NPC_Back
  }
  sub_nodes [ConversationChoiceNode_18 {
    class=ConversationChoiceNode
    properties {defaultChoiceDisplayText=BACK, choiceTags {gameplayTags [{tagName=Conversation.Choice.Category.Back}]}}
   }]
 }
 Speech_C_0 {
  id=#11
  class=Speech_C
  outputs [#13, #14]
  properties {
   softTaskIcon=/Game/Athena/HUD/Interaction/T_UI_NPC_Money
   generalConfig {contextualMessages [
     {
      message="ACE HAS TOLD ME I NEED TO VERIFY IDENTITIES, BUT YOU SEEM LEGIT."
      contextRequirements [{requirement {class=FortControllerRequirement_HasOwnedTag, properties {ownedTag {tagName=Conversation.Context.ServiceJustPurchased}}}}]
      metadataParameters [Style {value=Laugh}]
      priority {curve {curveTable=null, rowName=None}}
     }
     {
      message="GOODBYE, STRANGER WHO I TRUST IMPLICITLY! I'LL MISS YOU!"
      contextRequirements [{requirement {class=FortControllerRequirement_HasOwnedTag, properties {ownedTag {tagName=Conversation.Context.ServiceJustPurchased}}}}]
      priority {curve {curveTable=null, rowName=None}}
     }
     {
      message="YOU CAN TAKE THIS, I THINK. YOU BELONG TO THE COLD BLOODED, RIGHT?"
      contextRequirements []
      priority {value=0, curve {curveTable=null, rowName=None}}
     }
    ]}
  }
  sub_nodes [
   FortConversationChoiceWithPreviewNode_7 {class=FortConversationChoiceWithPreviewNode, properties {defaultChoiceDisplayText=BUY}}
   FortConversationRequirement_HasService_5 {class=FortConversationRequirement_HasService, properties {serviceTag {tagName=Tandem.Service.SellSubCategory1.Sell4}}}
  ]
 }
 SellLoot_C_2 {
  id=#14
  class=SellLoot_C
  outputs [#15]
  properties {sellSlot=4, optionalFilterCategoryTag {tagName=Tandem.Service.SellSubCategory1}}
  sub_nodes [SideEffect_PlayServiceUsedEffects_C_7 {class=SideEffect_PlayServiceUsedEffects_C}]
 }
 FortConversationTaskNode_ApplyDataDrivenEffect_0 {id=#15, class=FortConversationTaskNode_ApplyDataDrivenEffect, outputs [#16]}
 Repeat_C_0 {id=#16, class=Repeat_C}
 Back_C_1 {
  id=#13
  class=Back_C
  sub_nodes [ConversationChoiceNode_13 {class=ConversationChoiceNode, properties {defaultChoiceDisplayText=BACK}}]
 }
 Speech_C_6 {
  id=#10
  class=Speech_C
  outputs [#17, #18]
  properties {
   softTaskIcon=&T_UI_NPC_Money
   generalConfig {contextualMessages [
     {
      message="ACE HAS TOLD ME I NEED TO VERIFY IDENTITIES, BUT YOU SEEM LEGIT."
      contextRequirements [{requirement {class=FortControllerRequirement_HasOwnedTag, properties {ownedTag {tagName=Conversation.Context.ServiceJustPurchased}}}}]
      metadataParameters [Style {value=Laugh}]
      priority {curve {curveTable=null, rowName=None}}
     }
     {
      message="GOODBYE, STRANGER WHO I TRUST IMPLICITLY! I'LL MISS YOU!"
      contextRequirements [{requirement {class=FortControllerRequirement_HasOwnedTag, properties {ownedTag {tagName=Conversation.Context.ServiceJustPurchased}}}}]
      priority {curve {curveTable=null, rowName=None}}
     }
     {
      message="YOU CAN TAKE THIS, I THINK. YOU BELONG TO THE COLD BLOODED, RIGHT?"
      contextRequirements []
      priority {value=0, curve {curveTable=null, rowName=None}}
     }
    ]}
  }
  sub_nodes [
   FortConversationChoiceWithPreviewNode_6 {class=FortConversationChoiceWithPreviewNode, properties {defaultChoiceDisplayText=BUY}}
   FortConversationRequirement_HasService_4 {class=FortConversationRequirement_HasService, properties {serviceTag {tagName=Tandem.Service.SellSubCategory1.Sell3}}}
  ]
 }
 SellLoot_C_5 {
  id=#18
  class=SellLoot_C
  outputs [#15]
  properties {sellSlot=3, optionalFilterCategoryTag {tagName=Tandem.Service.SellSubCategory1}}
  sub_nodes [SideEffect_PlayServiceUsedEffects_C_6 {class=SideEffect_PlayServiceUsedEffects_C}]
 }
 Back_C_4 {
  id=#17
  class=Back_C
  sub_nodes [ConversationChoiceNode_12 {class=ConversationChoiceNode, properties {defaultChoiceDisplayText=BACK}}]
 }
 Speech_C_5 {
  id=#9
  class=Speech_C
  outputs [#19, #20]
  properties {
   softTaskIcon=&T_UI_NPC_Money
   generalConfig {contextualMessages [
     {
      message="ACE HAS TOLD ME I NEED TO VERIFY IDENTITIES, BUT YOU SEEM LEGIT."
      contextRequirements [{requirement {class=FortControllerRequirement_HasOwnedTag, properties {ownedTag {tagName=Conversation.Context.ServiceJustPurchased}}}}]
      metadataParameters [Style {value=Laugh}]
      priority {curve {curveTable=null, rowName=None}}
     }
     {
      message="GOODBYE, STRANGER WHO I TRUST IMPLICITLY! I'LL MISS YOU!"
      contextRequirements [{requirement {class=FortControllerRequirement_HasOwnedTag, properties {ownedTag {tagName=Conversation.Context.ServiceJustPurchased}}}}]
      priority {curve {curveTable=null, rowName=None}}
     }
     {
      message="YOU CAN TAKE THIS, I THINK. YOU BELONG TO THE COLD BLOODED, RIGHT?"
      contextRequirements []
      priority {value=0, curve {curveTable=null, rowName=None}}
     }
    ]}
  }
  sub_nodes [
   FortConversationChoiceWithPreviewNode_5 {class=FortConversationChoiceWithPreviewNode, properties {defaultChoiceDisplayText=BUY}}
   FortConversationRequirement_HasService_3 {class=FortConversationRequirement_HasService, properties {serviceTag {tagName=Tandem.Service.SellSubCategory1.Sell2}}}
  ]
 }
 SellLoot_C_1 {
  id=#20
  class=SellLoot_C
  outputs [#15]
  properties {sellSlot=2, optionalFilterCategoryTag {tagName=Tandem.Service.SellSubCategory1}}
  sub_nodes [SideEffect_PlayServiceUsedEffects_C_5 {class=SideEffect_PlayServiceUsedEffects_C}]
 }
 Back_C_3 {
  id=#19
  class=Back_C
  sub_nodes [ConversationChoiceNode_11 {class=ConversationChoiceNode, properties {defaultChoiceDisplayText=BACK}}]
 }
 Speech_C_4 {
  id=#8
  class=Speech_C
  outputs [#21, #22]
  properties {
   softTaskIcon=&T_UI_NPC_Money
   generalConfig {contextualMessages [
     {
      message="ACE HAS TOLD ME I NEED TO VERIFY IDENTITIES, BUT YOU SEEM LEGIT."
      contextRequirements [{requirement {class=FortControllerRequirement_HasOwnedTag, properties {ownedTag {tagName=Conversation.Context.ServiceJustPurchased}}}}]
      metadataParameters [Style {value=Laugh}]
      priority {curve {curveTable=null, rowName=None}}
     }
     {
      message="GOODBYE, STRANGER WHO I TRUST IMPLICITLY! I'LL MISS YOU!"
      contextRequirements [{requirement {class=FortControllerRequirement_HasOwnedTag, properties {ownedTag {tagName=Conversation.Context.ServiceJustPurchased}}}}]
      priority {curve {curveTable=null, rowName=None}}
     }
     {
      message="YOU CAN TAKE THIS, I THINK. YOU BELONG TO THE COLD BLOODED, RIGHT?"
      contextRequirements []
      priority {value=0, curve {curveTable=null, rowName=None}}
     }
    ]}
  }
  sub_nodes [
   FortConversationChoiceWithPreviewNode_3 {class=FortConversationChoiceWithPreviewNode, properties {defaultChoiceDisplayText=BUY}}
   FortConversationRequirement_HasService_2 {class=FortConversationRequirement_HasService, properties {serviceTag {tagName=Tandem.Service.SellSubCategory1.Sell1}}}
  ]
 }
 SellLoot_C_0 {
  id=#22
  class=SellLoot_C
  outputs [#15]
  properties {optionalFilterCategoryTag {tagName=Tandem.Service.SellSubCategory1}}
  sub_nodes [SideEffect_PlayServiceUsedEffects_C_2 {class=SideEffect_PlayServiceUsedEffects_C}]
 }
 Back_C_0 {
  id=#21
  class=Back_C
  sub_nodes [ConversationChoiceNode_5 {class=ConversationChoiceNode, properties {defaultChoiceDisplayText=BACK}}]
 }
 Speech_C_10 {
  id=#6
  class=Speech_C
  outputs [
   #23
   #24
   #25
   #26
   #27
  ]
  properties {
   softTaskIcon=/Game/Athena/HUD/Spectator/Art/KillFeed_Explosive
   generalConfig {contextualMessages [
     {message="MAKE A SELECTION, TRUSTED USER!", contextRequirements [], priority {value=0, curve {curveTable=null, rowName=None}}}
     {message="HOW CAN I ARM YOU TODAY, HONORABLE PATRON?", contextRequirements [], priority {value=0, curve {curveTable=null, rowName=None}}}
    ]}
  }
  sub_nodes [
   ConversationChoiceNode_10 {
    class=ConversationChoiceNode
    properties {defaultChoiceDisplayText=Explosives, choiceTags {gameplayTags [{tagName=Conversation.Choice.CustomSortPriority.Lowest}]}}
   }
   FortConversationRequirement_HasService_80 {class=FortConversationRequirement_HasService, properties {bRequireExactMatch=false, serviceTag {tagName=Tandem.Service.SellSubCategory5}}}
  ]
 }
 ReturnToStart_C_14 {
  id=#27
  class=ReturnToStart_C
  properties {
   choiceSound=&Conversation_Back_Cue
   hoverChoiceSound=&Conversation_Hover_Back_Cue
   softTaskIcon=&T_UI_NPC_Back
  }
  sub_nodes [ConversationChoiceNode_103 {
    class=ConversationChoiceNode
    properties {defaultChoiceDisplayText=BACK, choiceTags {gameplayTags [{tagName=Conversation.Choice.Category.Back}]}}
   }]
 }
 Speech_C_65 {
  id=#26
  class=Speech_C
  outputs [#28, #29]
  properties {
   softTaskIcon=&T_UI_NPC_Money
   generalConfig {contextualMessages [
     {
      message="ACE HAS TOLD ME I NEED TO VERIFY IDENTITIES, BUT YOU SEEM LEGIT."
      contextRequirements [{requirement {class=FortControllerRequirement_HasOwnedTag, properties {ownedTag {tagName=Conversation.Context.ServiceJustPurchased}}}}]
      metadataParameters [Style {value=Laugh}]
      priority {curve {curveTable=null, rowName=None}}
     }
     {
      message="GOODBYE, STRANGER WHO I TRUST IMPLICITLY! I'LL MISS YOU!"
      contextRequirements [{requirement {class=FortControllerRequirement_HasOwnedTag, properties {ownedTag {tagName=Conversation.Context.ServiceJustPurchased}}}}]
      priority {curve {curveTable=null, rowName=None}}
     }
     {
      message="HOW CAN I SAY NO TO THOSE SWEET, INNOCENT, EXPLOSIVE-LOVIN' EYES?"
      contextRequirements []
      priority {value=0, curve {curveTable=null, rowName=None}}
     }
    ]}
  }
  sub_nodes [
   FortConversationChoiceWithPreviewNode_75 {class=FortConversationChoiceWithPreviewNode, properties {defaultChoiceDisplayText=BUY}}
   FortConversationRequirement_HasService_77 {class=FortConversationRequirement_HasService, properties {serviceTag {tagName=Tandem.Service.SellSubCategory5.Sell4}}}
  ]
 }
 SellLoot_C_62 {
  id=#29
  class=SellLoot_C
  outputs [#30]
  properties {sellSlot=4, optionalFilterCategoryTag {tagName=Tandem.Service.SellSubCategory5}}
  sub_nodes [SideEffect_PlayServiceUsedEffects_C_78 {class=SideEffect_PlayServiceUsedEffects_C}]
 }
 FortConversationTaskNode_ApplyDataDrivenEffect_15 {id=#30, class=FortConversationTaskNode_ApplyDataDrivenEffect, outputs [#31]}
 Repeat_C_15 {id=#31, class=Repeat_C}
 Back_C_63 {
  id=#28
  class=Back_C
  sub_nodes [ConversationChoiceNode_102 {class=ConversationChoiceNode, properties {defaultChoiceDisplayText=BACK}}]
 }
 Speech_C_64 {
  id=#25
  class=Speech_C
  outputs [#32, #33]
  properties {
   softTaskIcon=&T_UI_NPC_Money
   generalConfig {contextualMessages [
     {
      message="ACE HAS TOLD ME I NEED TO VERIFY IDENTITIES, BUT YOU SEEM LEGIT."
      contextRequirements [{requirement {class=FortControllerRequirement_HasOwnedTag, properties {ownedTag {tagName=Conversation.Context.ServiceJustPurchased}}}}]
      metadataParameters [Style {value=Laugh}]
      priority {curve {curveTable=null, rowName=None}}
     }
     {
      message="GOODBYE, STRANGER WHO I TRUST IMPLICITLY! I'LL MISS YOU!"
      contextRequirements [{requirement {class=FortControllerRequirement_HasOwnedTag, properties {ownedTag {tagName=Conversation.Context.ServiceJustPurchased}}}}]
      priority {curve {curveTable=null, rowName=None}}
     }
     {
      message="HOW CAN I SAY NO TO THOSE SWEET, INNOCENT, EXPLOSIVE-LOVIN' EYES?"
      contextRequirements []
      priority {value=0, curve {curveTable=null, rowName=None}}
     }
    ]}
  }
  sub_nodes [
   FortConversationChoiceWithPreviewNode_74 {class=FortConversationChoiceWithPreviewNode, properties {defaultChoiceDisplayText=BUY}}
   FortConversationRequirement_HasService_76 {class=FortConversationRequirement_HasService, properties {serviceTag {tagName=Tandem.Service.SellSubCategory5.Sell3}}}
  ]
 }
 SellLoot_C_61 {
  id=#33
  class=SellLoot_C
  outputs [#30]
  properties {sellSlot=3, optionalFilterCategoryTag {tagName=Tandem.Service.SellSubCategory5}}
  sub_nodes [SideEffect_PlayServiceUsedEffects_C_77 {class=SideEffect_PlayServiceUsedEffects_C}]
 }
 Back_C_62 {
  id=#32
  class=Back_C
  sub_nodes [ConversationChoiceNode_101 {class=ConversationChoiceNode, properties {defaultChoiceDisplayText=BACK}}]
 }
 Speech_C_63 {
  id=#24
  class=Speech_C
  outputs [#34, #35]
  properties {
   softTaskIcon=&T_UI_NPC_Money
   generalConfig {contextualMessages [
     {
      message="ACE HAS TOLD ME I NEED TO VERIFY IDENTITIES, BUT YOU SEEM LEGIT."
      contextRequirements [{requirement {class=FortControllerRequirement_HasOwnedTag, properties {ownedTag {tagName=Conversation.Context.ServiceJustPurchased}}}}]
      metadataParameters [Style {value=Laugh}]
      priority {curve {curveTable=null, rowName=None}}
     }
     {
      message="GOODBYE, STRANGER WHO I TRUST IMPLICITLY! I'LL MISS YOU!"
      contextRequirements [{requirement {class=FortControllerRequirement_HasOwnedTag, properties {ownedTag {tagName=Conversation.Context.ServiceJustPurchased}}}}]
      priority {curve {curveTable=null, rowName=None}}
     }
     {
      message="HOW CAN I SAY NO TO THOSE SWEET, INNOCENT, EXPLOSIVE-LOVIN' EYES?"
      contextRequirements []
      priority {value=0, curve {curveTable=null, rowName=None}}
     }
    ]}
  }
  sub_nodes [
   FortConversationChoiceWithPreviewNode_73 {class=FortConversationChoiceWithPreviewNode, properties {defaultChoiceDisplayText=BUY}}
   FortConversationRequirement_HasService_75 {class=FortConversationRequirement_HasService, properties {serviceTag {tagName=Tandem.Service.SellSubCategory5.Sell2}}}
  ]
 }
 SellLoot_C_60 {
  id=#35
  class=SellLoot_C
  outputs [#30]
  properties {sellSlot=2, optionalFilterCategoryTag {tagName=Tandem.Service.SellSubCategory5}}
  sub_nodes [SideEffect_PlayServiceUsedEffects_C_76 {class=SideEffect_PlayServiceUsedEffects_C}]
 }
 Back_C_61 {
  id=#34
  class=Back_C
  sub_nodes [ConversationChoiceNode_100 {class=ConversationChoiceNode, properties {defaultChoiceDisplayText=BACK}}]
 }
 Speech_C_62 {
  id=#23
  class=Speech_C
  outputs [#36, #37]
  properties {
   softTaskIcon=&T_UI_NPC_Money
   generalConfig {contextualMessages [
     {
      message="ACE HAS TOLD ME I NEED TO VERIFY IDENTITIES, BUT YOU SEEM LEGIT."
      contextRequirements [{requirement {class=FortControllerRequirement_HasOwnedTag, properties {ownedTag {tagName=Conversation.Context.ServiceJustPurchased}}}}]
      metadataParameters [Style {value=Laugh}]
      priority {curve {curveTable=null, rowName=None}}
     }
     {
      message="GOODBYE, STRANGER WHO I TRUST IMPLICITLY! I'LL MISS YOU!"
      contextRequirements [{requirement {class=FortControllerRequirement_HasOwnedTag, properties {ownedTag {tagName=Conversation.Context.ServiceJustPurchased}}}}]
      priority {curve {curveTable=null, rowName=None}}
     }
     {
      message="HOW CAN I SAY NO TO THOSE SWEET, INNOCENT, EXPLOSIVE-LOVIN' EYES?"
      contextRequirements []
      priority {value=0, curve {curveTable=null, rowName=None}}
     }
    ]}
  }
  sub_nodes [
   FortConversationChoiceWithPreviewNode_72 {class=FortConversationChoiceWithPreviewNode, properties {defaultChoiceDisplayText=BUY}}
   FortConversationRequirement_HasService_74 {class=FortConversationRequirement_HasService, properties {serviceTag {tagName=Tandem.Service.SellSubCategory5.Sell1}}}
  ]
 }
 SellLoot_C_59 {
  id=#37
  class=SellLoot_C
  outputs [#30]
  properties {optionalFilterCategoryTag {tagName=Tandem.Service.SellSubCategory5}}
  sub_nodes [SideEffect_PlayServiceUsedEffects_C_75 {class=SideEffect_PlayServiceUsedEffects_C}]
 }
 Back_C_60 {
  id=#36
  class=Back_C
  sub_nodes [ConversationChoiceNode_99 {class=ConversationChoiceNode, properties {defaultChoiceDisplayText=BACK}}]
 }
 Speech_C_8 {
  id=#5
  class=Speech_C
  outputs [
   #38
   #39
   #40
   #41
   #42
  ]
  properties {
   softTaskIcon=/Game/UI/Foundation/Textures/Icons/ItemTypes/T-Icon-Sniper-128
   generalConfig {contextualMessages [
     {message="MAKE A SELECTION, TRUSTED USER!", contextRequirements [], priority {value=0, curve {curveTable=null, rowName=None}}}
     {message="HOW CAN I ARM YOU TODAY, HONORABLE PATRON?", contextRequirements [], priority {value=0, curve {curveTable=null, rowName=None}}}
    ]}
  }
  sub_nodes [
   ConversationChoiceNode_9 {
    class=ConversationChoiceNode
    properties {defaultChoiceDisplayText="Heavy Guns && Ammo", choiceTags {gameplayTags [{tagName=Conversation.Choice.CustomSortPriority.Low}]}}
   }
   FortConversationRequirement_HasService_79 {class=FortConversationRequirement_HasService, properties {bRequireExactMatch=false, serviceTag {tagName=Tandem.Service.SellSubCategory4}}}
  ]
 }
 ReturnToStart_C_11 {
  id=#42
  class=ReturnToStart_C
  properties {
   choiceSound=&Conversation_Back_Cue
   hoverChoiceSound=&Conversation_Hover_Back_Cue
   softTaskIcon=&T_UI_NPC_Back
  }
  sub_nodes [ConversationChoiceNode_74 {
    class=ConversationChoiceNode
    properties {defaultChoiceDisplayText=BACK, choiceTags {gameplayTags [{tagName=Conversation.Choice.Category.Back}]}}
   }]
 }
 Speech_C_53 {
  id=#41
  class=Speech_C
  outputs [#43, #44]
  properties {
   softTaskIcon=&T_UI_NPC_Money
   generalConfig {contextualMessages [
     {
      message="ACE HAS TOLD ME I NEED TO VERIFY IDENTITIES, BUT YOU SEEM LEGIT."
      contextRequirements [{requirement {class=FortControllerRequirement_HasOwnedTag, properties {ownedTag {tagName=Conversation.Context.ServiceJustPurchased}}}}]
      metadataParameters [Style {value=Laugh}]
      priority {curve {curveTable=null, rowName=None}}
     }
     {
      message="GOODBYE, STRANGER WHO I TRUST IMPLICITLY! I'LL MISS YOU!"
      contextRequirements [{requirement {class=FortControllerRequirement_HasOwnedTag, properties {ownedTag {tagName=Conversation.Context.ServiceJustPurchased}}}}]
      priority {curve {curveTable=null, rowName=None}}
     }
     {
      message="I'LL SELL THIS TO YOU, BUT YOU HAVE TO PROMISE YOU WORK FOR ACE. "
      contextRequirements []
      priority {value=0, curve {curveTable=null, rowName=None}}
     }
    ]}
  }
  sub_nodes [
   FortConversationChoiceWithPreviewNode_52 {class=FortConversationChoiceWithPreviewNode, properties {defaultChoiceDisplayText=BUY}}
   FortConversationRequirement_HasService_54 {class=FortConversationRequirement_HasService, properties {serviceTag {tagName=Tandem.Service.SellSubCategory4.Sell4}}}
  ]
 }
 SellLoot_C_50 {
  id=#44
  class=SellLoot_C
  outputs [#45]
  properties {sellSlot=4, optionalFilterCategoryTag {tagName=Tandem.Service.SellSubCategory4}}
  sub_nodes [SideEffect_PlayServiceUsedEffects_C_55 {class=SideEffect_PlayServiceUsedEffects_C}]
 }
 FortConversationTaskNode_ApplyDataDrivenEffect_12 {id=#45, class=FortConversationTaskNode_ApplyDataDrivenEffect, outputs [#46]}
 Repeat_C_12 {id=#46, class=Repeat_C}
 Back_C_51 {
  id=#43
  class=Back_C
  sub_nodes [ConversationChoiceNode_75 {class=ConversationChoiceNode, properties {defaultChoiceDisplayText=BACK}}]
 }
 Speech_C_52 {
  id=#40
  class=Speech_C
  outputs [#47, #48]
  properties {
   softTaskIcon=&T_UI_NPC_Money
   generalConfig {contextualMessages [
     {
      message="ACE HAS TOLD ME I NEED TO VERIFY IDENTITIES, BUT YOU SEEM LEGIT."
      contextRequirements [{requirement {class=FortControllerRequirement_HasOwnedTag, properties {ownedTag {tagName=Conversation.Context.ServiceJustPurchased}}}}]
      metadataParameters [Style {value=Laugh}]
      priority {curve {curveTable=null, rowName=None}}
     }
     {
      message="GOODBYE, STRANGER WHO I TRUST IMPLICITLY! I'LL MISS YOU!"
      contextRequirements [{requirement {class=FortControllerRequirement_HasOwnedTag, properties {ownedTag {tagName=Conversation.Context.ServiceJustPurchased}}}}]
      priority {curve {curveTable=null, rowName=None}}
     }
     {
      message="I'LL SELL THIS TO YOU, BUT YOU HAVE TO PROMISE YOU WORK FOR ACE. "
      contextRequirements []
      priority {value=0, curve {curveTable=null, rowName=None}}
     }
    ]}
  }
  sub_nodes [
   FortConversationChoiceWithPreviewNode_53 {class=FortConversationChoiceWithPreviewNode, properties {defaultChoiceDisplayText=BUY}}
   FortConversationRequirement_HasService_55 {class=FortConversationRequirement_HasService, properties {serviceTag {tagName=Tandem.Service.SellSubCategory4.Sell3}}}
  ]
 }
 SellLoot_C_49 {
  id=#48
  class=SellLoot_C
  outputs [#45]
  properties {sellSlot=3, optionalFilterCategoryTag {tagName=Tandem.Service.SellSubCategory4}}
  sub_nodes [SideEffect_PlayServiceUsedEffects_C_56 {class=SideEffect_PlayServiceUsedEffects_C}]
 }
 Back_C_50 {
  id=#47
  class=Back_C
  sub_nodes [ConversationChoiceNode_76 {class=ConversationChoiceNode, properties {defaultChoiceDisplayText=BACK}}]
 }
 Speech_C_51 {
  id=#39
  class=Speech_C
  outputs [#49, #50]
  properties {
   softTaskIcon=&T_UI_NPC_Money
   generalConfig {contextualMessages [
     {
      message="ACE HAS TOLD ME I NEED TO VERIFY IDENTITIES, BUT YOU SEEM LEGIT."
      contextRequirements [{requirement {class=FortControllerRequirement_HasOwnedTag, properties {ownedTag {tagName=Conversation.Context.ServiceJustPurchased}}}}]
      metadataParameters [Style {value=Laugh}]
      priority {curve {curveTable=null, rowName=None}}
     }
     {
      message="GOODBYE, STRANGER WHO I TRUST IMPLICITLY! I'LL MISS YOU!"
      contextRequirements [{requirement {class=FortControllerRequirement_HasOwnedTag, properties {ownedTag {tagName=Conversation.Context.ServiceJustPurchased}}}}]
      priority {curve {curveTable=null, rowName=None}}
     }
     {
      message="I'LL SELL THIS TO YOU, BUT YOU HAVE TO PROMISE YOU WORK FOR ACE. "
      contextRequirements []
      priority {value=0, curve {curveTable=null, rowName=None}}
     }
    ]}
  }
  sub_nodes [
   FortConversationChoiceWithPreviewNode_54 {class=FortConversationChoiceWithPreviewNode, properties {defaultChoiceDisplayText=BUY}}
   FortConversationRequirement_HasService_56 {class=FortConversationRequirement_HasService, properties {serviceTag {tagName=Tandem.Service.SellSubCategory4.Sell2}}}
  ]
 }
 SellLoot_C_48 {
  id=#50
  class=SellLoot_C
  outputs [#45]
  properties {sellSlot=2, optionalFilterCategoryTag {tagName=Tandem.Service.SellSubCategory4}}
  sub_nodes [SideEffect_PlayServiceUsedEffects_C_57 {class=SideEffect_PlayServiceUsedEffects_C}]
 }
 Back_C_49 {
  id=#49
  class=Back_C
  sub_nodes [ConversationChoiceNode_77 {class=ConversationChoiceNode, properties {defaultChoiceDisplayText=BACK}}]
 }
 Speech_C_50 {
  id=#38
  class=Speech_C
  outputs [#51, #52]
  properties {
   softTaskIcon=&T_UI_NPC_Money
   generalConfig {contextualMessages [
     {
      message="ACE HAS TOLD ME I NEED TO VERIFY IDENTITIES, BUT YOU SEEM LEGIT."
      contextRequirements [{requirement {class=FortControllerRequirement_HasOwnedTag, properties {ownedTag {tagName=Conversation.Context.ServiceJustPurchased}}}}]
      metadataParameters [Style {value=Laugh}]
      priority {curve {curveTable=null, rowName=None}}
     }
     {
      message="GOODBYE, STRANGER WHO I TRUST IMPLICITLY! I'LL MISS YOU!"
      contextRequirements [{requirement {class=FortControllerRequirement_HasOwnedTag, properties {ownedTag {tagName=Conversation.Context.ServiceJustPurchased}}}}]
      priority {curve {curveTable=null, rowName=None}}
     }
     {
      message="I'LL SELL THIS TO YOU, BUT YOU HAVE TO PROMISE YOU WORK FOR ACE. "
      contextRequirements []
      priority {value=0, curve {curveTable=null, rowName=None}}
     }
    ]}
  }
  sub_nodes [
   FortConversationChoiceWithPreviewNode_55 {class=FortConversationChoiceWithPreviewNode, properties {defaultChoiceDisplayText=BUY}}
   FortConversationRequirement_HasService_57 {class=FortConversationRequirement_HasService, properties {serviceTag {tagName=Tandem.Service.SellSubCategory4.Sell1}}}
  ]
 }
 SellLoot_C_64 {
  id=#52
  class=SellLoot_C
  outputs [#45]
  properties {optionalFilterCategoryTag {tagName=Tandem.Service.SellSubCategory4}}
  sub_nodes [SideEffect_PlayServiceUsedEffects_C_80 {class=SideEffect_PlayServiceUsedEffects_C}]
 }
 Back_C_48 {
  id=#51
  class=Back_C
  sub_nodes [ConversationChoiceNode_78 {class=ConversationChoiceNode, properties {defaultChoiceDisplayText=BACK}}]
 }
 Speech_C_9 {
  id=#4
  class=Speech_C
  outputs [
   #53
   #54
   #55
   #56
   #57
  ]
  properties {
   softTaskIcon=/Game/UI/Foundation/Textures/Icons/ItemTypes/T-Icon-Shotgun-128
   generalConfig {contextualMessages [
     {message="MAKE A SELECTION, TRUSTED USER!", contextRequirements [], priority {value=0, curve {curveTable=null, rowName=None}}}
     {message="HOW CAN I ARM YOU TODAY, HONORABLE PATRON?", contextRequirements [], priority {value=0, curve {curveTable=null, rowName=None}}}
    ]}
  }
  sub_nodes [
   ConversationChoiceNode_8 {
    class=ConversationChoiceNode
    properties {defaultChoiceDisplayText="Shotshell Guns && Ammo", choiceTags {gameplayTags [{tagName=Conversation.Choice.CustomSortPriority.Medium}]}}
   }
   FortConversationRequirement_HasService_78 {class=FortConversationRequirement_HasService, properties {bRequireExactMatch=false, serviceTag {tagName=Tandem.Service.SellSubCategory3}}}
  ]
 }
 ReturnToStart_C_13 {
  id=#57
  class=ReturnToStart_C
  properties {
   choiceSound=&Conversation_Back_Cue
   hoverChoiceSound=&Conversation_Hover_Back_Cue
   softTaskIcon=&T_UI_NPC_Back
  }
  sub_nodes [ConversationChoiceNode_94 {
    class=ConversationChoiceNode
    properties {defaultChoiceDisplayText=BACK, choiceTags {gameplayTags [{tagName=Conversation.Choice.Category.Back}]}}
   }]
 }
 Speech_C_58 {
  id=#56
  class=Speech_C
  outputs [#58, #59]
  properties {
   softTaskIcon=&T_UI_NPC_Money
   generalConfig {contextualMessages [
     {
      message="ACE HAS TOLD ME I NEED TO VERIFY IDENTITIES, BUT YOU SEEM LEGIT."
      contextRequirements [{requirement {class=FortControllerRequirement_HasOwnedTag, properties {ownedTag {tagName=Conversation.Context.ServiceJustPurchased}}}}]
      metadataParameters [Style {value=Laugh}]
      priority {curve {curveTable=null, rowName=None}}
     }
     {
      message="GOODBYE, STRANGER WHO I TRUST IMPLICITLY! I'LL MISS YOU!"
      contextRequirements [{requirement {class=FortControllerRequirement_HasOwnedTag, properties {ownedTag {tagName=Conversation.Context.ServiceJustPurchased}}}}]
      priority {curve {curveTable=null, rowName=None}}
     }
     {
      message="THIS ITEM IS FOR COLD BLOODED ONLY. LIKE YOU?! RIGHT?"
      contextRequirements []
      priority {value=0, curve {curveTable=null, rowName=None}}
     }
    ]}
  }
  sub_nodes [
   FortConversationChoiceWithPreviewNode_68 {class=FortConversationChoiceWithPreviewNode, properties {defaultChoiceDisplayText=BUY}}
   FortConversationRequirement_HasService_70 {class=FortConversationRequirement_HasService, properties {serviceTag {tagName=Tandem.Service.SellSubCategory3.Sell4}}}
  ]
 }
 SellLoot_C_55 {
  id=#59
  class=SellLoot_C
  outputs [#60]
  properties {sellSlot=4, optionalFilterCategoryTag {tagName=Tandem.Service.SellSubCategory3}}
  sub_nodes [SideEffect_PlayServiceUsedEffects_C_71 {class=SideEffect_PlayServiceUsedEffects_C}]
 }
 FortConversationTaskNode_ApplyDataDrivenEffect_14 {id=#60, class=FortConversationTaskNode_ApplyDataDrivenEffect, outputs [#61]}
 Repeat_C_14 {id=#61, class=Repeat_C}
 Back_C_56 {
  id=#58
  class=Back_C
  sub_nodes [ConversationChoiceNode_95 {class=ConversationChoiceNode, properties {defaultChoiceDisplayText=BACK}}]
 }
 Speech_C_59 {
  id=#55
  class=Speech_C
  outputs [#62, #63]
  properties {
   softTaskIcon=&T_UI_NPC_Money
   generalConfig {contextualMessages [
     {
      message="ACE HAS TOLD ME I NEED TO VERIFY IDENTITIES, BUT YOU SEEM LEGIT."
      contextRequirements [{requirement {class=FortControllerRequirement_HasOwnedTag, properties {ownedTag {tagName=Conversation.Context.ServiceJustPurchased}}}}]
      metadataParameters [Style {value=Laugh}]
      priority {curve {curveTable=null, rowName=None}}
     }
     {
      message="GOODBYE, STRANGER WHO I TRUST IMPLICITLY! I'LL MISS YOU!"
      contextRequirements [{requirement {class=FortControllerRequirement_HasOwnedTag, properties {ownedTag {tagName=Conversation.Context.ServiceJustPurchased}}}}]
      priority {curve {curveTable=null, rowName=None}}
     }
     {
      message="THIS ITEM IS FOR COLD BLOODED ONLY. LIKE YOU?! RIGHT?"
      contextRequirements []
      priority {value=0, curve {curveTable=null, rowName=None}}
     }
    ]}
  }
  sub_nodes [
   FortConversationChoiceWithPreviewNode_69 {class=FortConversationChoiceWithPreviewNode, properties {defaultChoiceDisplayText=BUY}}
   FortConversationRequirement_HasService_71 {class=FortConversationRequirement_HasService, properties {serviceTag {tagName=Tandem.Service.SellSubCategory3.Sell3}}}
  ]
 }
 SellLoot_C_56 {
  id=#63
  class=SellLoot_C
  outputs [#60]
  properties {sellSlot=3, optionalFilterCategoryTag {tagName=Tandem.Service.SellSubCategory3}}
  sub_nodes [SideEffect_PlayServiceUsedEffects_C_72 {class=SideEffect_PlayServiceUsedEffects_C}]
 }
 Back_C_57 {
  id=#62
  class=Back_C
  sub_nodes [ConversationChoiceNode_96 {class=ConversationChoiceNode, properties {defaultChoiceDisplayText=BACK}}]
 }
 Speech_C_60 {
  id=#54
  class=Speech_C
  outputs [#64, #65]
  properties {
   softTaskIcon=&T_UI_NPC_Money
   generalConfig {contextualMessages [
     {
      message="ACE HAS TOLD ME I NEED TO VERIFY IDENTITIES, BUT YOU SEEM LEGIT."
      contextRequirements [{requirement {class=FortControllerRequirement_HasOwnedTag, properties {ownedTag {tagName=Conversation.Context.ServiceJustPurchased}}}}]
      metadataParameters [Style {value=Laugh}]
      priority {curve {curveTable=null, rowName=None}}
     }
     {
      message="GOODBYE, STRANGER WHO I TRUST IMPLICITLY! I'LL MISS YOU!"
      contextRequirements [{requirement {class=FortControllerRequirement_HasOwnedTag, properties {ownedTag {tagName=Conversation.Context.ServiceJustPurchased}}}}]
      priority {curve {curveTable=null, rowName=None}}
     }
     {
      message="THIS ITEM IS FOR COLD BLOODED ONLY. LIKE YOU?! RIGHT?"
      contextRequirements []
      priority {value=0, curve {curveTable=null, rowName=None}}
     }
    ]}
  }
  sub_nodes [
   FortConversationChoiceWithPreviewNode_70 {class=FortConversationChoiceWithPreviewNode, properties {defaultChoiceDisplayText=BUY}}
   FortConversationRequirement_HasService_72 {class=FortConversationRequirement_HasService, properties {serviceTag {tagName=Tandem.Service.SellSubCategory3.Sell2}}}
  ]
 }
 SellLoot_C_57 {
  id=#65
  class=SellLoot_C
  outputs [#60]
  properties {sellSlot=2, optionalFilterCategoryTag {tagName=Tandem.Service.SellSubCategory3}}
  sub_nodes [SideEffect_PlayServiceUsedEffects_C_73 {class=SideEffect_PlayServiceUsedEffects_C}]
 }
 Back_C_58 {
  id=#64
  class=Back_C
  sub_nodes [ConversationChoiceNode_97 {class=ConversationChoiceNode, properties {defaultChoiceDisplayText=BACK}}]
 }
 Speech_C_61 {
  id=#53
  class=Speech_C
  outputs [#66, #67]
  properties {
   softTaskIcon=&T_UI_NPC_Money
   generalConfig {contextualMessages [
     {
      message="ACE HAS TOLD ME I NEED TO VERIFY IDENTITIES, BUT YOU SEEM LEGIT."
      contextRequirements [{requirement {class=FortControllerRequirement_HasOwnedTag, properties {ownedTag {tagName=Conversation.Context.ServiceJustPurchased}}}}]
      metadataParameters [Style {value=Laugh}]
      priority {curve {curveTable=null, rowName=None}}
     }
     {
      message="GOODBYE, STRANGER WHO I TRUST IMPLICITLY! I'LL MISS YOU!"
      contextRequirements [{requirement {class=FortControllerRequirement_HasOwnedTag, properties {ownedTag {tagName=Conversation.Context.ServiceJustPurchased}}}}]
      priority {curve {curveTable=null, rowName=None}}
     }
     {
      message="THIS ITEM IS FOR COLD BLOODED ONLY. LIKE YOU?! RIGHT?"
      contextRequirements []
      priority {value=0, curve {curveTable=null, rowName=None}}
     }
    ]}
  }
  sub_nodes [
   FortConversationRequirement_HasService_73 {class=FortConversationRequirement_HasService, properties {serviceTag {tagName=Tandem.Service.SellSubCategory3.Sell1}}}
   FortConversationChoiceWithPreviewNode_71 {class=FortConversationChoiceWithPreviewNode, properties {defaultChoiceDisplayText=BUY}}
  ]
 }
 SellLoot_C_63 {
  id=#67
  class=SellLoot_C
  outputs [#60]
  properties {optionalFilterCategoryTag {tagName=Tandem.Service.SellSubCategory3}}
  sub_nodes [SideEffect_PlayServiceUsedEffects_C_79 {class=SideEffect_PlayServiceUsedEffects_C}]
 }
 Back_C_59 {
  id=#66
  class=Back_C
  sub_nodes [ConversationChoiceNode_98 {class=ConversationChoiceNode, properties {defaultChoiceDisplayText=BACK}}]
 }
 Speech_C_7 {
  id=#3
  class=Speech_C
  outputs [
   #68
   #69
   #70
   #71
   #72
  ]
  properties {
   softTaskIcon=/Game/UI/Foundation/Textures/Icons/ItemTypes/T-Icon-Assault-128
   generalConfig {contextualMessages [
     {message="MAKE A SELECTION, TRUSTED USER!", contextRequirements [], priority {value=0, curve {curveTable=null, rowName=None}}}
     {message="HOW CAN I ARM YOU TODAY, HONORABLE PATRON?", contextRequirements [], priority {value=0, curve {curveTable=null, rowName=None}}}
    ]}
  }
  sub_nodes [
   ConversationChoiceNode_7 {
    class=ConversationChoiceNode
    properties {defaultChoiceDisplayText="Medium Guns && Ammo", choiceTags {gameplayTags [{tagName=Conversation.Choice.CustomSortPriority.High}]}}
   }
   FortConversationRequirement_HasService_1 {class=FortConversationRequirement_HasService, properties {bRequireExactMatch=false, serviceTag {tagName=Tandem.Service.SellSubCategory2}}}
  ]
 }
 ReturnToStart_C_10 {
  id=#72
  class=ReturnToStart_C
  properties {
   choiceSound=&Conversation_Back_Cue
   hoverChoiceSound=&Conversation_Hover_Back_Cue
   softTaskIcon=&T_UI_NPC_Back
  }
  sub_nodes [ConversationChoiceNode_83 {
    class=ConversationChoiceNode
    properties {defaultChoiceDisplayText=BACK, choiceTags {gameplayTags [{tagName=Conversation.Choice.Category.Back}]}}
   }]
 }
 Speech_C_46 {
  id=#71
  class=Speech_C
  outputs [#73, #74]
  properties {
   softTaskIcon=&T_UI_NPC_Money
   generalConfig {contextualMessages [
     {
      message="ACE HAS TOLD ME I NEED TO VERIFY IDENTITIES, BUT YOU SEEM LEGIT."
      contextRequirements [{requirement {class=FortControllerRequirement_HasOwnedTag, properties {ownedTag {tagName=Conversation.Context.ServiceJustPurchased}}}}]
      metadataParameters [Style {value=Laugh}]
      priority {curve {curveTable=null, rowName=None}}
     }
     {
      message="GOODBYE, STRANGER WHO I TRUST IMPLICITLY! I'LL MISS YOU!"
      contextRequirements [{requirement {class=FortControllerRequirement_HasOwnedTag, properties {ownedTag {tagName=Conversation.Context.ServiceJustPurchased}}}}]
      priority {curve {curveTable=null, rowName=None}}
     }
     {
      message="ONLY COLD BLOODED CAN BUY THESE... BUT I WORK ON THE HONOR SYSTEM ;)"
      contextRequirements []
      priority {value=0, curve {curveTable=null, rowName=None}}
     }
    ]}
  }
  sub_nodes [
   FortConversationRequirement_HasService_61 {class=FortConversationRequirement_HasService, properties {serviceTag {tagName=Tandem.Service.SellSubCategory2.Sell4}}}
   FortConversationChoiceWithPreviewNode_59 {class=FortConversationChoiceWithPreviewNode, properties {defaultChoiceDisplayText=BUY}}
  ]
 }
 SellLoot_C_43 {
  id=#74
  class=SellLoot_C
  outputs [#75]
  properties {sellSlot=4, optionalFilterCategoryTag {tagName=Tandem.Service.SellSubCategory2}}
  sub_nodes [SideEffect_PlayServiceUsedEffects_C_62 {class=SideEffect_PlayServiceUsedEffects_C}]
 }
 FortConversationTaskNode_ApplyDataDrivenEffect_11 {id=#75, class=FortConversationTaskNode_ApplyDataDrivenEffect, outputs [#76]}
 Repeat_C_11 {id=#76, class=Repeat_C}
 Back_C_44 {
  id=#73
  class=Back_C
  sub_nodes [ConversationChoiceNode_82 {class=ConversationChoiceNode, properties {defaultChoiceDisplayText=BACK}}]
 }
 Speech_C_47 {
  id=#70
  class=Speech_C
  outputs [#77, #78]
  properties {
   softTaskIcon=&T_UI_NPC_Money
   generalConfig {contextualMessages [
     {
      message="ACE HAS TOLD ME I NEED TO VERIFY IDENTITIES, BUT YOU SEEM LEGIT."
      contextRequirements [{requirement {class=FortControllerRequirement_HasOwnedTag, properties {ownedTag {tagName=Conversation.Context.ServiceJustPurchased}}}}]
      metadataParameters [Style {value=Laugh}]
      priority {curve {curveTable=null, rowName=None}}
     }
     {
      message="GOODBYE, STRANGER WHO I TRUST IMPLICITLY! I'LL MISS YOU!"
      contextRequirements [{requirement {class=FortControllerRequirement_HasOwnedTag, properties {ownedTag {tagName=Conversation.Context.ServiceJustPurchased}}}}]
      priority {curve {curveTable=null, rowName=None}}
     }
     {
      message="ONLY COLD BLOODED CAN BUY THESE... BUT I WORK ON THE HONOR SYSTEM ;)"
      contextRequirements []
      priority {value=0, curve {curveTable=null, rowName=None}}
     }
    ]}
  }
  sub_nodes [
   FortConversationChoiceWithPreviewNode_58 {class=FortConversationChoiceWithPreviewNode, properties {defaultChoiceDisplayText=BUY}}
   FortConversationRequirement_HasService_60 {class=FortConversationRequirement_HasService, properties {serviceTag {tagName=Tandem.Service.SellSubCategory2.Sell3}}}
  ]
 }
 SellLoot_C_44 {
  id=#78
  class=SellLoot_C
  outputs [#75]
  properties {sellSlot=3, optionalFilterCategoryTag {tagName=Tandem.Service.SellSubCategory2}}
  sub_nodes [SideEffect_PlayServiceUsedEffects_C_61 {class=SideEffect_PlayServiceUsedEffects_C}]
 }
 Back_C_45 {
  id=#77
  class=Back_C
  sub_nodes [ConversationChoiceNode_81 {class=ConversationChoiceNode, properties {defaultChoiceDisplayText=BACK}}]
 }
 Speech_C_48 {
  id=#69
  class=Speech_C
  outputs [#79, #80]
  properties {
   softTaskIcon=&T_UI_NPC_Money
   generalConfig {contextualMessages [
     {
      message="ACE HAS TOLD ME I NEED TO VERIFY IDENTITIES, BUT YOU SEEM LEGIT."
      contextRequirements [{requirement {class=FortControllerRequirement_HasOwnedTag, properties {ownedTag {tagName=Conversation.Context.ServiceJustPurchased}}}}]
      metadataParameters [Style {value=Laugh}]
      priority {curve {curveTable=null, rowName=None}}
     }
     {
      message="GOODBYE, STRANGER WHO I TRUST IMPLICITLY! I'LL MISS YOU!"
      contextRequirements [{requirement {class=FortControllerRequirement_HasOwnedTag, properties {ownedTag {tagName=Conversation.Context.ServiceJustPurchased}}}}]
      priority {curve {curveTable=null, rowName=None}}
     }
     {
      message="ONLY COLD BLOODED CAN BUY THESE... BUT I WORK ON THE HONOR SYSTEM ;)"
      contextRequirements []
      priority {value=0, curve {curveTable=null, rowName=None}}
     }
    ]}
  }
  sub_nodes [
   FortConversationChoiceWithPreviewNode_57 {class=FortConversationChoiceWithPreviewNode, properties {defaultChoiceDisplayText=BUY}}
   FortConversationRequirement_HasService_59 {class=FortConversationRequirement_HasService, properties {serviceTag {tagName=Tandem.Service.SellSubCategory2.Sell2}}}
  ]
 }
 SellLoot_C_45 {
  id=#80
  class=SellLoot_C
  outputs [#75]
  properties {sellSlot=2, optionalFilterCategoryTag {tagName=Tandem.Service.SellSubCategory2}}
  sub_nodes [SideEffect_PlayServiceUsedEffects_C_60 {class=SideEffect_PlayServiceUsedEffects_C}]
 }
 Back_C_46 {
  id=#79
  class=Back_C
  sub_nodes [ConversationChoiceNode_80 {class=ConversationChoiceNode, properties {defaultChoiceDisplayText=BACK}}]
 }
 Speech_C_49 {
  id=#68
  class=Speech_C
  outputs [#81, #82]
  properties {
   softTaskIcon=&T_UI_NPC_Money
   generalConfig {contextualMessages [
     {
      message="ACE HAS TOLD ME I NEED TO VERIFY IDENTITIES, BUT YOU SEEM LEGIT."
      contextRequirements [{requirement {class=FortControllerRequirement_HasOwnedTag, properties {ownedTag {tagName=Conversation.Context.ServiceJustPurchased}}}}]
      metadataParameters [Style {value=Laugh}]
      priority {curve {curveTable=null, rowName=None}}
     }
     {
      message="GOODBYE, STRANGER WHO I TRUST IMPLICITLY! I'LL MISS YOU!"
      contextRequirements [{requirement {class=FortControllerRequirement_HasOwnedTag, properties {ownedTag {tagName=Conversation.Context.ServiceJustPurchased}}}}]
      priority {curve {curveTable=null, rowName=None}}
     }
     {
      message="ONLY COLD BLOODED CAN BUY THESE... BUT I WORK ON THE HONOR SYSTEM ;)"
      contextRequirements []
      priority {value=0, curve {curveTable=null, rowName=None}}
     }
    ]}
  }
  sub_nodes [
   FortConversationChoiceWithPreviewNode_56 {class=FortConversationChoiceWithPreviewNode, properties {defaultChoiceDisplayText=BUY}}
   FortConversationRequirement_HasService_58 {class=FortConversationRequirement_HasService, properties {serviceTag {tagName=Tandem.Service.SellSubCategory2.Sell1}}}
  ]
 }
 SellLoot_C_46 {
  id=#82
  class=SellLoot_C
  outputs [#75]
  properties {optionalFilterCategoryTag {tagName=Tandem.Service.SellSubCategory2}}
  sub_nodes [SideEffect_PlayServiceUsedEffects_C_59 {class=SideEffect_PlayServiceUsedEffects_C}]
 }
 Back_C_47 {
  id=#81
  class=Back_C
  sub_nodes [ConversationChoiceNode_79 {class=ConversationChoiceNode, properties {defaultChoiceDisplayText=BACK}}]
 }
]
